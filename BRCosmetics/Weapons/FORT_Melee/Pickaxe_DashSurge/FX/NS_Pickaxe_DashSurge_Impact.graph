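# Niagara impact effect for the DashSurge pickaxe, with six looks chosen by a System.Variant index.
# Visible emitters include flash sprites with a light, time-of-day-tinted smoke, ice shards and distance-culled "leek" pieces.
# Variant switches pick colors, while per-variant bool arrays and comparisons gate extras like glow, bokeh, galaxy and shards.
# Front-end modifiers and a FrontEndScale user parameter adjust sizes and motion for menu display.
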NiagaraSystem /BRCosmetics/Weapons/FORT_Melee/Pickaxe_DashSurge/FX/NS_Pickaxe_DashSurge_Impact
dependencies {
 usedDataInterfaces [
  &NiagaraDataInterfaceCurve
  &NiagaraDataInterfaceColorCurve
  &NiagaraDataInterfaceMeshRendererInfo
  /Script/Niagara.NiagaraDataInterfaceArrayBool
 ]
 usedDynamicInputs [&MakeIntFromFloat]
 usedModules [
  &self:SystemSpawnScript.SystemScriptSource.SystemScriptGraph.NiagaraNodeAssignment_0.SetVariables_#1
  &self:SystemSpawnScript.SystemScriptSource.SystemScriptGraph.NiagaraNodeAssignment_1.SetVariables_#2
  &SystemState
 ]
}
emitters {
 Dark_Glow {
  emitterSpawnScript [
   /SetVariables(Emitter.UseGlow:NiagaraBool=/Niagara/DynamicInputs/Arrays/SelectBoolFromArray(
     Array Sampling Mode:ENiagara_LifetimeMode=::NewEnumerator0
     Direct Array Index:NiagaraInt32=$System.Variant
     Direct Array Mode:ENiagara_WrapClamp=::NewEnumerator1
     Evaluation Type:ENiagaraRandomnessEvaluation=::NewEnumerator0
     Recalculate Random Each Loop:NiagaraBool=false
     Randomness Mode:ENiagaraRandomnessMode=::NewEnumerator0
     Random Seed:NiagaraInt32=?
     Fixed Random Seed:NiagaraBool=false
     Bool Selection Array:NiagaraDataInterfaceArrayBool=DataInterface() {propertyValues {GpuSyncMode=SyncCpuToGpu, MaxElements=0, BoolData [false,false,true,false,true,true]}}
    )) {isSetParameters=true}
   /SetVariables(
    Emitter.UseBokeh:NiagaraBool=&SelectBoolFromArray(
     Array Sampling Mode:ENiagara_LifetimeMode=::NewEnumerator0
     Direct Array Index:NiagaraInt32=$System.Variant
     Direct Array Mode:ENiagara_WrapClamp=::NewEnumerator1
     Evaluation Type:ENiagaraRandomnessEvaluation=::NewEnumerator0
     Recalculate Random Each Loop:NiagaraBool=false
     Randomness Mode:ENiagaraRandomnessMode=::NewEnumerator0
     Random Seed:NiagaraInt32=?
     Fixed Random Seed:NiagaraBool=false
     Bool Selection Array:NiagaraDataInterfaceArrayBool=DataInterface() {propertyValues {GpuSyncMode=SyncCpuToGpu, MaxElements=0, BoolData [false,false,false,false,true,false]}}
    )
    Emitter.UseGalaxy:NiagaraBool=&SelectBoolFromArray(
     Array Sampling Mode:ENiagara_LifetimeMode=::NewEnumerator0
     Direct Array Index:NiagaraInt32=$System.Variant
     Direct Array Mode:ENiagara_WrapClamp=::NewEnumerator1
     Evaluation Type:ENiagaraRandomnessEvaluation=::NewEnumerator0
     Recalculate Random Each Loop:NiagaraBool=false
     Randomness Mode:ENiagaraRandomnessMode=::NewEnumerator0
     Random Seed:NiagaraInt32=?
     Fixed Random Seed:NiagaraBool=false
     Bool Selection Array:NiagaraDataInterfaceArrayBool=DataInterface() {propertyValues {GpuSyncMode=SyncCpuToGpu, MaxElements=0, BoolData [false,false,false,false,false,true]}}
    )
   ) {isSetParameters=true}
   &self:Activate_Emitter(Active:NiagaraBool=&Boolean_OrOperation(Bool1=$Emitter.UseBokeh, Bool2=&Boolean_OrOperation(Bool1=$Emitter.UseGalaxy, Bool2=$Emitter.UseGlow))) {enabled=false, module=Activate_Emitter}
  ]
  emitterUpdateScript [
   &EmitterState(Life Cycle Mode=::NewEnumerator1, Loop Behavior=::NewEnumerator1, Loop Duration=1, UseLoopDelay=false)
   &SpawnBurst_Instantaneous(
    Spawn Count=&MakeCustomIntFromBool(Bool=$Emitter.UseGlow, True INT=1, False INT=0)
    Use Spawn Probability=false
   )
   &SpawnBurst_Instantaneous(
    Spawn Count=&MakeCustomIntFromBool(Bool=$Emitter.UseGlow, True INT=1, False INT=0)
    Spawn Group=1
    Use Spawn Probability=false
    UseLoopCountLimit=false
   ) {module=SpawnBurst_Instantaneous001}
   &SpawnBurst_Instantaneous(
    Spawn Count=&MakeCustomIntFromBool(Bool=$Emitter.UseGalaxy, True INT=2, False INT=0)
    Spawn Group=2
    Use Spawn Probability=false
    UseLoopCountLimit=false
   ) {module=SpawnBurst_Instantaneous002}
   &SpawnBurst_Instantaneous(
    Spawn Count=&MakeCustomIntFromBool(Bool=$Emitter.UseBokeh, True INT=&RandomRangeInt(Minimum=3, Maximum=4), False INT=0)
    Spawn Group=3
    Use Spawn Probability=false
    UseLoopCountLimit=false
   ) {module=SpawnBurst_Instantaneous003}
  ]
  particleSpawnScript [
   /SetVariables(Particles.SpawnGroup=$Emitter.SpawnGroup) {isSetParameters=true}
   &InitializeParticle(
    Color=&MergeHelper_ColorSwitch(
     0=<rgba 0.021 0.0139 0.004 3>
     1=&MergeHelper_ColorSwitch(
      0=<rgba 0.1771 0.1135 0 5>
      1=<rgba 0.1771 0.1135 0 5>
      2=<rgba 0.1771 0.1135 0 3>
      3=<rgba 0.1771 0.1135 0 5>
      4=<rgba 0.0364 0.1213 0.125 1>
      5=<rgba 0.0487 0.2133 0.3073 1>
      Index=$System.Variant
      Num=::NewEnumerator4
     )
     2=?
     3=?
     Index=$Particles.SpawnGroup
     Num=::NewEnumerator0
    )
    Lifetime=&RandomRangeFloat(Minimum=0.5, Maximum=0.7)
    Sprite Size Mode=::NewEnumerator3
    Uniform Sprite Size=&Multiply_Float(A=200, B=&FN_FrontEndFloatModifier(Gameplay=1, Front End=$User.FrontEndScale))
    UsePositionOffset=false
    Write Material Random=true
   ) {module=InitializeParticle001}
   &DynamicMaterialParameters(
    Index 0 Param 1=1
    Index 0 Param 2=1
    Index 0 Param 3=1
    Index 0 Param 4=1
    Module.Param0WriteEnabled=true
    Module.Param1WriteEnabled=true
    Module.Param2WriteEnabled=true
    Module.Param3WriteEnabled=true
   )
   /SetVariables(Particles.VisibilityTag=$Particles.SpawnGroup) {isSetParameters=true}
   &ShapeLocation(
    Apply To Particle Position=::NewEnumerator1
    MaskSpawnGroup=false
    Override Seed=false
    Sphere Radius=&Multiply_Float(A=5, B=&FN_FrontEndFloatModifier(Gameplay=1, Front End=$User.FrontEndScale))
    UseOwnerScale=true
   )
   &ShapeLocation(
    Apply To Particle Position=::NewEnumerator1
    MaskSpawnGroup=false
    Override Seed=false
    Sphere Radius=&Multiply_Float(A=15, B=&FN_FrontEndFloatModifier(Gameplay=1, Front End=$User.FrontEndScale))
    UseOwnerScale=true
   ) {module=ShapeLocation001}
   /SetVariables(
    Particles.Galaxy Color:LinearColor=&MakeLinearColorFromVectorAndFloat(
     "Vector (RGB)"=&MakeVectorFromLinearColorRGB(Color=<rgba 0.2717 1.2829 1.3 0.5>)
     "Float (Alpha)"=&RandomRangeFloat(Minimum=0.4, Maximum=0.6)
    )
    Particles.Galaxy Position:NiagaraPosition=$Particles.ShapeLocation.ShapePosition
    Particles.Galaxy Size:Vector2f=&Vector2DFromFloat(Value=&Multiply_Float(
      A=&RandomRangeFloat(Minimum=80, Maximum=100)
      B=&FN_FrontEndFloatModifier(Gameplay=1, Front End=$User.FrontEndScale)
     ))
   ) {isSetParameters=true}
   &CameraOffset(Camera Offset Amount=5) {module=CameraOffset001}
   /SetVariables(
    Particles.Bokeh Color:LinearColor=&RandomRangeLinearColor(Minimum=<rgba 0.6771 0.6771 0.6771 1>, Maximum=<rgba 1 1 1 1>)
    Particles.Bokeh Position:NiagaraPosition=$Particles.ShapeLocation001.ShapePosition
    Particles.Bokeh Size:Vector2f=&Vector2DFromFloat(Value=&Multiply_Float(
      A=&RandomRangeFloat(Minimum=15, Maximum=35)
      B=&FN_FrontEndFloatModifier(Gameplay=1, Front End=$User.FrontEndScale)
     ))
    Particles.Bokeh Velocity:Vector3f=&Multiply_VectorByFloat(Vector=<1 1 1>, Float=&RandomRangeFloat(Minimum=0.2, Maximum=0.5))
    Particles.DynamicBokeh:Vector4f=<3 1 0 0>
   ) {isSetParameters=true}
  ]
  particleUpdateScript [
   &CameraOffset(Camera Offset Amount=20)
   &ParticleState()
   &ScaleColor(
    Scale Alpha=&Multiply_Float(
     A=&FloatFromCurve(
      FloatCurve=DataInterface() {propertyValues {Curve {keys [<at 0 1>,<at 1 0>]}}}
      Scale Curve=&FN_FrontEndFloatModifier(Gameplay=1, Front End=0.5)
     )
     B=&FNTimeOfDayScalar(Night Scale=0.3, Morning Scale=0.05, Day Scale=0.5, Evening Scale=0.05)
    )
    ScaleA=true
    ScaleRGB=true
   )
   &ScaleSpriteSize(Scale Factor=&MergeHelper_Vector2Switch(
     0=&Vector2DFromFloat(Value=&FloatFromCurve(
       FloatCurve=DataInterface() {propertyValues {Curve {keys [<at 0 1>,<at 1 1.15>]}}}
       Scale Curve=&FN_FrontEndFloatModifier(Gameplay=1, Front End=0.5)
      ))
     1=&Vector2DFromFloat(Value=&FloatFromCurve(
       FloatCurve=DataInterface() {propertyValues {Curve {keys [{arriveTangent=1.8012, interpMode=RCIM_Cubic, leaveTangent=1.8012, tangentMode=RCTM_User, time=0, value=0.4}, <at 1 1.075>]}}}
       Scale Curve=&FN_FrontEndFloatModifier(Gameplay=1, Front End=0.5)
      ))
     2=?
     3=?
     4=?
     Index=$Particles.SpawnGroup
     Num=::NewEnumerator0
    ))
   /SetVariables(Particles.CoronaColor:LinearColor=&MultiplyLinearColor(
     Linear Color=$Particles.Initial.Color
     Scale Factor=&MakeVector4_FromLinearColor(LinearColor=&ColorFromCurve(ColorCurve=DataInterface() {propertyValues {
         AlphaCurve {keys [<at 0 1.1 RCIM_Cubic>,<at 1 0 RCIM_Cubic>]}
         BlueCurve {keys [<at 0 3 RCIM_Cubic>]}
         GreenCurve {keys [<at 0 3 RCIM_Cubic>]}
         RedCurve {keys [<at 0 3 RCIM_Cubic>]}
        }}))
    )) {isSetParameters=true}
   /SetVariables(
    Particles.Galaxy Color:LinearColor=&MultiplyLinearColor(
     Linear Color=&MakeLinearColorFromVectorAndFloat(
      "Vector (RGB)"=<1 1 1>
      "Float (Alpha)"=&FloatFromCurve(FloatCurve=DataInterface() {propertyValues {Curve {keys [
           {arriveTangent=-0.2253, interpMode=RCIM_Cubic, time=0, value=1}
           {arriveTangent=-1.6895, interpMode=RCIM_Cubic, leaveTangent=-1.6895, tangentMode=RCTM_User, time=1, value=0}
          ]}}})
     )
     Scale Factor=&MakeVector4_FromLinearColor(LinearColor=$Particles.Initial.Galaxy Color)
    )
    Particles.Galaxy Size:Vector2f=&Multiply_Vector2D_ByFloat(
     Float=&FloatFromCurve(FloatCurve=DataInterface() {propertyValues {Curve {keys [
          {arriveTangent=3.1206, interpMode=RCIM_Cubic, leaveTangent=3.1206, tangentMode=RCTM_User, time=0, value=0}
          {interpMode=RCIM_Cubic, tangentMode=RCTM_User, time=0.9, value=1}
         ]}}})
     Vector2D=$Particles.Initial.Galaxy Size
    )
   ) {isSetParameters=true}
   /SetVariables(
    Particles.Bokeh Color:LinearColor=&MultiplyLinearColor(
     Linear Color=&MakeLinearColorFromVectorAndFloat(
      "Vector (RGB)"=&MakeVectorFromLinearColorRGB(Color=&ColorFromCurve(
        ColorCurve=DataInterface() {propertyValues {
          AlphaCurve {keys [<at 0 1 RCIM_Cubic>]}
          BlueCurve {keys [
            <at 0.0257 1>
            <at 0.2335 0.4193 RCIM_Cubic>
            {arriveTangent=0.9835, interpMode=RCIM_Cubic, leaveTangent=0.9835, time=0.571, value=1}
            <at 0.824 1>
            <at 0.9742 1>
           ]}
          GreenCurve {keys [
            <at 0.0257 0.3437>
            {arriveTangent=-0.2101, interpMode=RCIM_Cubic, leaveTangent=-0.2101, time=0.2335, value=1}
            {arriveTangent=-0.3331, interpMode=RCIM_Cubic, leaveTangent=-0.3331, time=0.571, value=0.2292}
            <at 0.824 0.8033>
            <at 0.9742 0.3437>
           ]}
          RedCurve {keys [
            <at 0.0257 0.1562>
            {arriveTangent=1.204, interpMode=RCIM_Cubic, leaveTangent=1.204, time=0.2335, value=0.1302}
            {arriveTangent=0.0265, interpMode=RCIM_Cubic, leaveTangent=0.0265, time=0.571, value=0.8127}
            <at 0.824 0.1458>
            <at 0.9742 0.1562>
           ]}
         }}
        CurveIndex=&Frac_Float(Float=&Add_Float(
          A=&Multiply_Float(A=$Particles.NormalizedAge, B=0.3)
          B=&RandomRangeFloat(Minimum=0, Maximum=1)
         ))
       ))
      "Float (Alpha)"=&FloatFromCurve(
       FloatCurve=DataInterface() {propertyValues {Curve {keys [{interpMode=RCIM_Cubic, leaveTangent=-2.4411, tangentMode=RCTM_Break, time=0.3, value=1}, <at 1 0>]}}}
       Scale Curve=&RandomRangeFloat(Minimum=0.1, Maximum=0.2)
      )
     )
     Scale Factor=&MakeVector4_FromLinearColor(LinearColor=$Particles.Initial.Bokeh Color)
    )
    Particles.Bokeh Position:NiagaraPosition=&AddVectorToPosition(Position=$Particles.Bokeh Position, Vector=<0 0 0>)
    Particles.Bokeh Size:Vector2f=&Multiply_Vector2D_ByFloat(
     Float=&FloatFromCurve(FloatCurve=DataInterface() {propertyValues {Curve {keys [
          {arriveTangent=5.8866, interpMode=RCIM_Cubic, leaveTangent=5.8866, tangentMode=RCTM_User, time=0, value=0}
          {arriveTangent=1.3, interpMode=RCIM_Cubic, leaveTangent=1.3, time=0.3, value=1}
          <at 1 1.3 RCIM_Cubic>
         ]}}})
     Vector2D=$Particles.Initial.Bokeh Size
    )
   ) {isSetParameters=true}
  ]
  rendererClasses [&NiagaraSpriteRendererProperties]
  renderers [
   {rendererClass=&NiagaraSpriteRendererProperties}
   {rendererClass=&NiagaraSpriteRendererProperties, rendererIndex=1}
   {rendererClass=&NiagaraSpriteRendererProperties, rendererIndex=2}
   {rendererClass=&NiagaraSpriteRendererProperties, rendererIndex=3}
  ]
 }
 Flash_Elements {
  emitterUpdateScript [
   &EmitterState()
   &SpawnBurst_Instantaneous(Spawn Group=1, Use Spawn Probability=false, UseLoopCountLimit=false) {module=SpawnBurst_Instantaneous001}
   &SpawnBurst_Instantaneous(
    Spawn Count=&RandomRangeInt(Minimum=3, Maximum=5)
    Use Spawn Probability=false
    UseLoopCountLimit=false
   )
  ]
  particleSpawnScript [
   &InitializeParticle(Color Mode=::NewEnumerator0, Lifetime=0.25, Sprite Rotation Mode=::NewEnumerator2, UsePositionOffset=false, Write Material Random=true)
   &ShapeLocation(MaskSpawnGroup=false, Override Seed=false, Sphere Radius=1, UseOwnerScale=true)
   &SpriteFacingAndAlignment(Sprite Alignment=$Particles.ShapeLocation.ShapeNormal, WriteAlignment=true, WriteFacing=false)
   &FN_PerParticleLOD(Enable SpriteSize Distance Scaling=true, Sprite Size Scale Far=<2.5 2.5>)
   /SetVariables(Particles.SpriteSize=<1 1>, Particles.VisibilityTag=&SpawnGroup()) {isSetParameters=true}
   /SetVariables(
    Particles.Flash Spike Color:LinearColor=&MergeHelper_ColorSwitch(
     0=<rgba 0.614 1 0 1>
     1=<rgba 1 0.0882 0 1>
     2=<rgba 1 0.4021 0 1>
     3=<rgba 0 1 0.9907 1>
     4=<rgba 1 0.237 0.8868 1>
     5=<rgba 0.733 0.1051 1 1>
     Index=$System.Variant
     Num=::NewEnumerator4
    )
    Particles.Flash Spike Lifetime:NiagaraFloat=&RandomRangeFloat(Minimum=0.05, Maximum=0.07)
    Particles.Flash Spike Sprite Size:Vector2f=&FN_FrontEndVector2Modifier(
     Front End:Vector2f=&RandomRangeVector2D(Minimum=<2 25>, Maximum=<7 40>)
     Gameplay:Vector2f=&RandomRangeVector2D(Minimum=<3 40>, Maximum=<7 60>)
    )
   ) {isSetParameters=true}
   /SetVariables(
    Particles.Flash Star Cam Offset=15
    Particles.Flash Star Lifetime=0.15
    Particles.Flash Star Rotation:NiagaraFloat=&RandomRangeFloat(Minimum=0, Maximum=360)
    Particles.Flash Star Sprite Size:Vector2f=&Multiply_Vector2D_ByFloat(
     Float=&Multiply_Float(
      A=&MakeCustomFloatFromBool(Bool=$User.bIsFrontEnd, True Float=0.45, False Float=1)
      B=&RandomRangeFloat(Minimum=100, Maximum=120)
     )
     Vector2D=$Particles.SpriteSize
    )
    Particles.Flash Star Sprite Size Alt:Vector2f=&Multiply_Vector2D_ByFloat(
     Float=&Multiply_Float(
      A=0.85
      B=&Multiply_Float(
       A=&MakeCustomFloatFromBool(Bool=$User.bIsFrontEnd, True Float=0.45, False Float=1)
       B=&RandomRangeFloat(Minimum=120, Maximum=150)
      )
     )
     Vector2D=$Particles.SpriteSize
    )
    Particles.Flash Star SubUV Index:NiagaraFloat=&RandomRangeFloat(Minimum=2, Maximum=3)
    Particles.Flash Star SubUV Index Alt:NiagaraFloat=&RandomRangeFloat(Minimum=0, Maximum=3)
   ) {isSetParameters=true}
   /SetVariables(
    Particles.Flash Flare Color:LinearColor=<rgba 3 3 3 0.7>
    Particles.Flash Flare Lifetime:NiagaraFloat=&FN_FrontEndFloatModifier(Gameplay=0.2, Front End=0.125)
    Particles.Flash Flare Sprite Size:Vector2f=&FN_FrontEndVector2Modifier(Front End:Vector2f=<110 110>, Gameplay:Vector2f=<220 220>)
   ) {isSetParameters=true}
   &DynamicMaterialParameters(
    Index 0 Param 1=0.2
    Index 0 Param 2=1
    Module.Index0Param0Write=true
    Module.Index0Param1Write=true
    Module.Index0Param2Write=true
    Module.Index0Param3Write=true
   )
   &CameraOffset(Camera Offset Amount=10)
  ]
  particleUpdateScript [
   &ParticleState()
   /SetVariables(
    Particles.Update Flash Flare Age:NiagaraFloat=&Divide_Float(A=$Particles.Age, B=$Particles.Initial.Flash Flare Lifetime)
    Particles.Update Flash Spike Age:NiagaraFloat=&Divide_Float(A=$Particles.Age, B=$Particles.Initial.Flash Spike Lifetime)
    Particles.Update Flash Star Age:NiagaraFloat=&Divide_Float(A=$Particles.Age, B=$Particles.Initial.Flash Star Lifetime)
   ) {isSetParameters=true}
   /SetVariables(
    Particles.Update Flash Spike Color:LinearColor=&MakeLinearColorFromVectorAndFloat(
     "Vector (RGB)"=&MakeVectorFromLinearColorRGB(Color=$Particles.Initial.Flash Spike Color)
     "Float (Alpha)"=&FloatFromCurve(
      FloatCurve=DataInterface() {propertyValues {Curve {keys [<at 0.7 1>,<at 1 0>]}}}
      CurveIndex=$Particles.Update Flash Spike Age
      Scale Curve=0.5
     )
    )
    Particles.Update Flash Spike Sprite Size:Vector2f=&Multiply_Vector2D_ByFloat(
     Float=&FloatFromCurve(FloatCurve=DataInterface() {propertyValues {Curve {keys [<at 0 0.6>,<at 1 2>]}}}, CurveIndex=$Particles.Update Flash Spike Age)
     Vector2D=$Particles.Initial.Flash Spike Sprite Size
    )
    Particles.Update Flash Star Color:LinearColor=&Multiply_LinearColorByFloat(
     Linear Color=&MergeHelper_ColorSwitch(
      0=&ColorFromCurve(ColorCurve=DataInterface() {propertyValues {
         AlphaCurve {keys [<at 0 1 RCIM_Cubic>,<at 0.9595 0>]}
         BlueCurve {keys [<at 0.0827 0 RCIM_Cubic>,<at 0.6886 0.0975 RCIM_Cubic>]}
         GreenCurve {keys [<at 0.0827 1 RCIM_Cubic>,<at 0.6886 0.6562 RCIM_Cubic>]}
         RedCurve {keys [<at 0.0827 0.567 RCIM_Cubic>,<at 0.6886 0.3048 RCIM_Cubic>]}
        }})
      1=&ColorFromCurve(ColorCurve=DataInterface() {propertyValues {
         AlphaCurve {keys [<at 0 1 RCIM_Cubic>,<at 0.9595 0>]}
         BlueCurve {keys [<at 0.0733 0 RCIM_Cubic>,<at 0.4468 0.0859 RCIM_Cubic>]}
         GreenCurve {keys [<at 0.0733 1 RCIM_Cubic>,<at 0.4468 0 RCIM_Cubic>]}
         RedCurve {keys [<at 0.0733 0.7582 RCIM_Cubic>,<at 0.4468 1 RCIM_Cubic>]}
        }})
      2=&ColorFromCurve(ColorCurve=DataInterface() {propertyValues {
         AlphaCurve {keys [<at 0 1 RCIM_Cubic>,<at 0.9577 0>]}
         BlueCurve {keys [<at 0.0185 0.4853>,<at 0.4626 0 RCIM_Cubic>]}
         GreenCurve {keys [<at 0.0185 0.7936>,<at 0.4626 0.5987 RCIM_Cubic>]}
         RedCurve {keys [<at 0.0185 1>,<at 0.4626 1 RCIM_Cubic>]}
        }})
      3=&ColorFromCurve(ColorCurve=DataInterface() {propertyValues {
         AlphaCurve {keys [<at 0 1 RCIM_Cubic>,<at 0.956 0>]}
         BlueCurve {keys [<at 0.0238 3>, {arriveTangent=-2.5249, interpMode=RCIM_Cubic, leaveTangent=-2.5249, time=0.2564, value=1}, <at 0.8159 1 RCIM_Cubic>]}
         GreenCurve {keys [
           <at 0.0238 2.4047>
           {arriveTangent=-2.3924, interpMode=RCIM_Cubic, leaveTangent=-2.3924, time=0.2564, value=0.9714}
           <at 0.8159 0.5097 RCIM_Cubic>
          ]}
         RedCurve {keys [<at 0.0238 0>,<at 0.2564 0 RCIM_Cubic>,<at 0.8159 0 RCIM_Cubic>]}
        }})
      4=&ColorFromCurve(ColorCurve=DataInterface() {propertyValues {
         AlphaCurve {keys [<at 0.0679 1 RCIM_Cubic>,<at 1 0>]}
         BlueCurve {keys [<at 0.1159 1 RCIM_Cubic>,<at 0.4549 0.6639>,<at 0.7704 1>,<at 0.9536 0.3783 RCIM_Cubic>]}
         GreenCurve {keys [<at 0.1159 0.1747 RCIM_Cubic>, <at 0.4549 1>, <at 0.7704 0.1962>, {interpMode=RCIM_Cubic, leaveTangent=-1.0692, time=0.9536, value=0.7026}]}
         RedCurve {keys [<at 0.1159 0.5276 RCIM_Cubic>, <at 0.4549 0.0599>, <at 0.7704 0.3449>, {interpMode=RCIM_Cubic, leaveTangent=0.607, time=0.9536, value=1}]}
        }})
      5=&ColorFromCurve(ColorCurve=DataInterface() {propertyValues {
         AlphaCurve {keys [<at 0 1 RCIM_Cubic>,<at 0.956 0>]}
         BlueCurve {keys [<at 0.0882 0.7046 RCIM_Cubic>,<at 0.4852 1 RCIM_Cubic>]}
         GreenCurve {keys [<at 0.0882 1 RCIM_Cubic>,<at 0.4852 0.107 RCIM_Cubic>]}
         RedCurve {keys [<at 0.0882 0 RCIM_Cubic>,<at 0.4852 0.3012 RCIM_Cubic>]}
        }})
      Index=$System.Variant
      Num=::NewEnumerator4
     )
     Float=0.8
    )
    Particles.Update Flash Star Sprize Size:Vector2f=&Multiply_Vector2D_ByFloat(
     Float=&FloatFromCurve(
      FloatCurve=DataInterface() {propertyValues {Curve {keys [
          {arriveTangent=2.7013, interpMode=RCIM_Cubic, leaveTangent=2.7013, tangentMode=RCTM_User, time=0, value=0}
          {interpMode=RCIM_Cubic, tangentMode=RCTM_User, time=1, value=1}
         ]}}}
      CurveIndex=$Particles.Update Flash Star Age
     )
     Vector2D=$Particles.Initial.Flash Star Sprite Size
    )
    Particles.Update Flash Star Sprize Size Alt:Vector2f=&Multiply_Vector2D_ByFloat(
     Float=&FloatFromCurve(
      FloatCurve=DataInterface() {propertyValues {Curve {keys [
          {arriveTangent=2.7013, interpMode=RCIM_Cubic, leaveTangent=2.7013, tangentMode=RCTM_User, time=0, value=0}
          {interpMode=RCIM_Cubic, tangentMode=RCTM_User, time=1, value=1}
         ]}}}
      CurveIndex=$Particles.Update Flash Star Age
     )
     Vector2D=$Particles.Initial.Flash Star Sprite Size Alt
    )
   ) {isSetParameters=true}
   /SetVariables(Particles.Update Flash Flare Color:LinearColor=&MakeLinearColorFromVectorAndFloat(
     "Vector (RGB)"=&Multiply_Vector(
      A=&Multiply_Vector(
       A=&MakeVectorFromLinearColorRGB(Color=$Particles.Initial.Flash Flare Color)
       B=&MakeVectorFromLinearColorRGB(Color=&MergeHelper_ColorSwitch(
         0=<rgba 0.2499 1 0 1>
         1=&ColorFromCurve(ColorCurve=DataInterface() {propertyValues {
            AlphaCurve {keys [<at 0 1 RCIM_Cubic>]}
            BlueCurve {keys [<at 0.0025 0.552>,<at 0.2273 0.4072>,<at 0.6067 0.0157 RCIM_Cubic>]}
            GreenCurve {keys [<at 0.0025 1>,<at 0.2273 0>,<at 0.6067 0 RCIM_Cubic>]}
            RedCurve {keys [<at 0.0025 0.9073>,<at 0.2273 2>,<at 0.6067 1 RCIM_Cubic>]}
           }})
         2=&ColorFromCurve(ColorCurve=DataInterface() {propertyValues {
            AlphaCurve {keys [<at 0 1>]}
            BlueCurve {keys [<at 0 0>,<at 0.3591 0>,<at 0.9851 0>]}
            GreenCurve {keys [<at 0 1.9488>,<at 0.3591 0.4923>,<at 0.9851 0.5903>]}
            RedCurve {keys [<at 0 2>,<at 0.3591 2>,<at 0.9851 2>]}
           }})
         3=<rgba 0 0.7302 1 1>
         4=&ColorFromCurve(ColorCurve=DataInterface() {propertyValues {
            AlphaCurve {keys [<at 0 1 RCIM_Cubic>]}
            BlueCurve {keys [<at 0.0757 0.3783 RCIM_Cubic>,<at 0.3257 1>,<at 0.6069 0.2401 RCIM_Cubic>,<at 1 1>]}
            GreenCurve {keys [
              {arriveTangent=-1.1111, interpMode=RCIM_Cubic, time=0.0757, value=0.7026}
              <at 0.3257 0.8547>
              {arriveTangent=-0.7421, interpMode=RCIM_Cubic, leaveTangent=-0.7421, time=0.6069, value=1}
              <at 1 0.3543>
             ]}
            RedCurve {keys [
              {arriveTangent=0.367, interpMode=RCIM_Cubic, time=0.0757, value=1}
              <at 0.3257 0.1667>
              {arriveTangent=0.4722, interpMode=RCIM_Cubic, leaveTangent=0.4722, time=0.6069, value=0.2138}
              <at 1 0.4851>
             ]}
           }})
         5=&ColorFromCurve(ColorCurve=DataInterface() {propertyValues {
            AlphaCurve {keys [<at 0 1 RCIM_Cubic>]}
            BlueCurve {keys [<at 0.0882 0.7046 RCIM_Cubic>,<at 0.8159 1 RCIM_Cubic>]}
            GreenCurve {keys [<at 0.0882 1 RCIM_Cubic>,<at 0.8159 0.107 RCIM_Cubic>]}
            RedCurve {keys [<at 0.0882 0 RCIM_Cubic>,<at 0.8159 0.3012 RCIM_Cubic>]}
           }})
         Index=$System.Variant
         Num=::NewEnumerator4
        ))
      )
      B=&VectorFromFloat(Value=&FN_FrontEndFloatModifier(Gameplay=1.5, Front End=1))
     )
     "Float (Alpha)"=&Multiply_Float(
      A=&MakeFloatFromLinearColor(LinearColor=<rgba 1 1 1 1>, Channel:ENiagaraLinearColor_Channels=::NewEnumerator3)
      B=&Lerp_Float(A=1, B=0, Alpha=$Particles.Update Flash Flare Age)
     )
    )) {isSetParameters=true}
   &FN_Light_Attributes(
    Color=&MergeHelper_ColorSwitch(
     0=<rgba 0.2499 1 0 1>
     1=&ColorFromCurve(ColorCurve=DataInterface() {propertyValues {
        AlphaCurve {keys [<at 0 1 RCIM_Cubic>]}
        BlueCurve {keys [<at 0.0023 0 RCIM_Cubic>,<at 0.5035 0.2271 RCIM_Cubic>]}
        GreenCurve {keys [<at 0.0023 1 RCIM_Cubic>,<at 0.5035 0 RCIM_Cubic>]}
        RedCurve {keys [<at 0.0023 0.5 RCIM_Cubic>,<at 0.5035 1 RCIM_Cubic>]}
       }})
     2=<rgba 1 0.4989 0 1>
     3=<rgba 0 0.7302 1 1>
     4=&ColorFromCurve(ColorCurve=DataInterface() {propertyValues {
        AlphaCurve {keys [<at 0 1 RCIM_Cubic>]}
        BlueCurve {keys [
          <at 0.0885 0.3783 RCIM_Cubic>
          <at 0.2969 1>
          {arriveTangent=0.7406, interpMode=RCIM_Cubic, leaveTangent=0.7406, time=0.6448, value=0.3323}
          <at 0.928 1>
         ]}
        GreenCurve {keys [
          {arriveTangent=-1.1116, interpMode=RCIM_Cubic, time=0.0885, value=0.7026}
          <at 0.2969 0.8802>
          {arriveTangent=-0.4148, interpMode=RCIM_Cubic, leaveTangent=-0.4148, time=0.6448, value=1}
          <at 0.928 0.3543>
         ]}
        RedCurve {keys [
          {arriveTangent=0.7581, interpMode=RCIM_Cubic, time=0.0885, value=1}
          <at 0.2969 0.1264>
          {arriveTangent=-0.6134, interpMode=RCIM_Cubic, leaveTangent=-0.6134, time=0.6448, value=0.1379}
          <at 0.928 0.4851>
         ]}
       }})
     5=&ColorFromCurve(ColorCurve=DataInterface() {propertyValues {
        AlphaCurve {keys [<at 0 1 RCIM_Cubic>]}
        BlueCurve {keys [<at 0.1824 0.7046 RCIM_Cubic>,<at 1 1 RCIM_Cubic>]}
        GreenCurve {keys [<at 0.1824 1 RCIM_Cubic>,<at 1 0.154 RCIM_Cubic>]}
        RedCurve {keys [<at 0.1824 0 RCIM_Cubic>,<at 1 0.437 RCIM_Cubic>]}
       }})
     Index=$System.Variant
     Num=::NewEnumerator4
    )
    Create Color Attribute=true
    Create Exponent Attribute=true
    Create Position Attribute=true
    Create Radius Attribute=true
    Exponent=5
    Position=<5 0 0>
    Radius=90
    Scale Color=&FloatFromCurve(
     FloatCurve=DataInterface() {propertyValues {Curve {keys [
         {interpMode=RCIM_Cubic, tangentMode=RCTM_User, time=0, value=1}
         {arriveTangent=-0.5426, interpMode=RCIM_Cubic, leaveTangent=-0.5426, tangentMode=RCTM_User, time=1, value=0}
        ]}}}
     Scale Curve=10
    )
   )
  ]
  rendererClasses [&NiagaraSpriteRendererProperties, &NiagaraLightRendererProperties]
  renderers [
   {rendererClass=&NiagaraSpriteRendererProperties}
   {rendererClass=&NiagaraSpriteRendererProperties, rendererIndex=1}
   {rendererClass=&NiagaraSpriteRendererProperties, rendererIndex=2}
   {rendererClass=&NiagaraSpriteRendererProperties, rendererIndex=3}
   {rendererClass=&NiagaraLightRendererProperties, rendererIndex=4}
  ]
 }
 Ice_Shards {
  emitterSpawnScript [
   /SetVariables(Emitter.SpawnParticles:NiagaraBool=&SetBoolByIntComparison(A=$System.Variant, B=3, Comparison Type=::NewEnumerator4)) {isSetParameters=true}
   &self:Activate_Emitter(Active:NiagaraBool=$Emitter.SpawnParticles) {module=Activate_Emitter}
  ]
  emitterUpdateScript [
   &EmitterState(Life Cycle Mode=::NewEnumerator1, Loop Behavior=::NewEnumerator1, Loop Duration=0.4, UseLoopDelay=false)
   &SpawnBurst_Instantaneous(
    Spawn Count=&MakeCustomIntFromBool(Bool=$Emitter.SpawnParticles, True INT=&RandomRangeInt(Minimum=3, Maximum=5), False INT=0)
    Use Spawn Probability=false
    UseLoopCountLimit=false
   )
  ]
  particleSpawnScript [
   &InitializeParticle(
    Color=<rgba 0.131 0.6319 1.5 1>
    Lifetime Max=0.35
    Lifetime Min=0.3
    Lifetime Mode=::NewEnumerator1
    Mesh Scale=&Multiply_VectorByFloat(
     Vector=<0.3 0.3 0.8>
     Float=&Multiply_Float(
      A=&RandomRangeFloat(Minimum=0.5, Maximum=1)
      B=&FN_FrontEndFloatModifier(Gameplay=1, Front End=$User.FrontEndScale)
     )
    )
    Mesh Scale Mode=::NewEnumerator4
    Sprite Size Mode=::NewEnumerator3
    Uniform Sprite Size=50
    UsePositionOffset=false
    Write Material Random=true
   )
   &ShapeLocation(
    MaskSpawnGroup=false
    Override Seed=false
    Sphere Radius=&Multiply_Float(A=5, B=&FN_FrontEndFloatModifier(Gameplay=0, Front End=$User.FrontEndScale))
    UseOwnerScale=true
   )
   &AddVelocity(
    Cone Angle=150
    Module.UseConeFalloff=false
    Module.UseVelDistribution=false
    Override Seed=false
    Velocity Mode=::NewEnumerator2
    Velocity Speed=&FN_FrontEndFloatModifier(Gameplay=1, Front End=$User.FrontEndScale)
   )
   &InitialMeshOrientation(Mesh Axis To Orient=<0 0 -1>, Mesh Orientation Mode=::NewEnumerator2, Use Rotation Vector=true)
   &CameraOffset(Camera Offset Amount=10)
  ]
  particleUpdateScript [
   &ParticleState()
   &ScaleColor(
    Scale Alpha=&FloatFromCurve(FloatCurve=DataInterface() {propertyValues {Curve {keys [<at 0.5 1 RCIM_Cubic>,<at 1.0011 0.002 RCIM_Cubic>]}}}, Scale Curve=0.25)
    Scale RGB=&VectorFromFloat(Value=&Multiply_Float(
      A=&RandomRangeFloat(Minimum=1, Maximum=2)
      B=&FNTimeOfDayScalar(Night Scale=0.5, Morning Scale=0.75, Evening Scale=0.75)
     ))
    ScaleA=true
    ScaleRGB=true
   )
   &ScaleMeshSize(Scale Factor=&VectorFromFloat(Value=&FloatFromCurve(
      FloatCurve=DataInterface() {propertyValues {Curve {keys [
          {arriveTangent=5.0353, interpMode=RCIM_Cubic, leaveTangent=5.0353, tangentMode=RCTM_User, time=0, value=0}
          {interpMode=RCIM_Cubic, tangentMode=RCTM_User, time=0.4, value=1}
          {tangentMode=RCTM_User, time=1, value=0}
         ]}}}
      Scale Curve=1.2
     )))
   /SetVariables(Particles.Position=&AddVectorToPosition(
     Position=$Particles.Initial.Position
     Vector=&Add_Vector(
      A=&Multiply_VectorByFloat(
       Vector=&Multiply_Vector(A=$Particles.Velocity, B=&FN_FrontEndVector3Modifier(Gameplay=<0.2 1 1>, Front End=<1 1 1>))
       Float=&FloatFromCurve(
        FloatCurve=DataInterface() {propertyValues {Curve {keys [
            {arriveTangent=0.3836, interpMode=RCIM_Cubic, leaveTangent=0.3836, tangentMode=RCTM_User, time=0, value=0}
            {arriveTangent=0.8891, tangentMode=RCTM_User, time=1, value=1}
           ]}}}
        Scale Curve=&Multiply_Float(
         A=&RandomRangeFloat(Minimum=50, Maximum=100)
         B=&FN_FrontEndFloatModifier(Gameplay=1, Front End=$User.FrontEndScale)
        )
       )
      )
      B=&Multiply_VectorByFloat(
       Vector=<0 0 -1>
       Float=&FloatFromCurve(
        FloatCurve=DataInterface() {propertyValues {Curve {keys [
            {arriveTangent=0.0327, interpMode=RCIM_Cubic, leaveTangent=0.0327, tangentMode=RCTM_User, time=0.3, value=0}
            {arriveTangent=2.3808, tangentMode=RCTM_User, time=1, value=1}
           ]}}}
        Scale Curve=&Multiply_Float(
         A=&RandomRangeFloat(Minimum=5, Maximum=10)
         B=&FN_FrontEndFloatModifier(Gameplay=1, Front End=$User.FrontEndScale)
        )
       )
      )
     )
    )) {isSetParameters=true}
   &DynamicMaterialParameters(
    Index 0 Param 1=&MakeFloatFromVector(Vector=$Particles.Velocity, Channel=::NewEnumerator0)
    Index 0 Param 2=&MakeFloatFromVector(Vector=$Particles.Velocity, Channel=::NewEnumerator1)
    Index 0 Param 3=&MakeFloatFromVector(Vector=$Particles.Velocity, Channel=::NewEnumerator2)
    Index 0 Param 4=&FloatFromCurve(
     FloatCurve=DataInterface() {propertyValues {Curve {keys [
         {arriveTangent=5.0353, interpMode=RCIM_Cubic, leaveTangent=5.0353, tangentMode=RCTM_User, time=0, value=0}
         {interpMode=RCIM_Cubic, tangentMode=RCTM_User, time=0.4, value=1}
         {tangentMode=RCTM_User, time=1, value=0}
        ]}}}
     Scale Curve=&Multiply_Float(A=120, B=&FN_FrontEndFloatModifier(Gameplay=1, Front End=1))
    )
    Module.Param0WriteEnabled=true
    Module.Param1WriteEnabled=true
    Module.Param2WriteEnabled=true
    Module.Param3WriteEnabled=true
   )
  ]
  rendererClasses [&NiagaraMeshRendererProperties, &NiagaraSpriteRendererProperties]
  renderers [
   {rendererClass=&NiagaraMeshRendererProperties, rendererIndex=0}
   {rendererClass=&NiagaraSpriteRendererProperties, rendererIndex=1}
  ]
 }
 Impact_Smoke {
  emitterUpdateScript [
   &EmitterState()
   &SpawnBurst_Instantaneous(
    Spawn Count=&RandomRangeInt(Minimum=3, Maximum=4)
    Use Spawn Probability=false
    UseLoopCountLimit=false
   )
  ]
  particleSpawnScript [
   &InitializeParticle(
    Color Maximum=<rgba 1 1 1 0.6>
    Color Minimum=<rgba 1 1 1 0.4>
    Color Mode=::NewEnumerator2
    Lifetime Max=0.5
    Lifetime Min=0.4
    Lifetime Mode=::NewEnumerator1
    Sprite Size Mode=::NewEnumerator3
    Uniform Sprite Size=&Multiply_Float(
     A=&MakeCustomFloatFromBool(Bool=$User.bIsFrontEnd, True Float=&Multiply_Float(A=$User.FrontEndScale, B=0.7), False Float=1)
     B=&RandomRangeFloat(Minimum=100, Maximum=130)
    )
    UsePositionOffset=false
    Write Material Random=true
   )
   &CameraOffset(Camera Offset Amount=&RandomRangeFloat(Minimum=40, Maximum=50, Evaluation Type=::NewEnumerator1))
   &ShapeLocation(
    Hemisphere Distribution=<1 0.5>
    MaskSpawnGroup=false
    Override Seed=false
    Rotation Angle=90
    Rotation Mode=::NewEnumerator1
    Sphere Radius=&MakeCustomFloatFromBool(Bool=$User.bIsFrontEnd, True Float=5, False Float=10)
    UseOwnerScale=true
   )
   &AddVelocity(Velocity=&Multiply_VectorByFloat(Vector=$Particles.ShapeLocation.ShapeVector, Float=10))
  ]
  particleUpdateScript [
   &ParticleState()
   &ScaleSpriteSize(Uniform Curve Sprite Scale=DataInterface() {propertyValues {Curve {keys [
        {arriveTangent=2.2125, interpMode=RCIM_Cubic, leaveTangent=2.2125, tangentMode=RCTM_User, time=0, value=0}
        {interpMode=RCIM_Cubic, tangentMode=RCTM_User, time=1, value=1}
       ]}}})
   &ScaleColor(
    Scale Alpha=&FloatFromCurve(FloatCurve=DataInterface() {propertyValues {Curve {keys [
         {arriveTangent=2.4358, interpMode=RCIM_Cubic, leaveTangent=2.4358, tangentMode=RCTM_User, time=0, value=0}
         <at 0.25 0.5 RCIM_Cubic>
         <at 1 0 RCIM_Cubic>
        ]}}})
    Scale RGB=&Multiply_Vector(
     A=&MakeVectorFromLinearColorRGB(Color=&MergeHelper_ColorSwitch(
       0=&ColorFromCurve(ColorCurve=DataInterface() {propertyValues {
          AlphaCurve {keys [<at 0 1 RCIM_Cubic>]}
          BlueCurve {keys [<at 0 0 RCIM_Cubic>, {interpMode=RCIM_Cubic, leaveTangent=0.1469, time=0.1492, value=0}, <at 0.6397 0.114>]}
          GreenCurve {keys [<at 0 0.8452 RCIM_Cubic>, {interpMode=RCIM_Cubic, leaveTangent=-0.6927, time=0.1492, value=1}, <at 0.6397 0.3073>]}
          RedCurve {keys [<at 0 1 RCIM_Cubic>, {interpMode=RCIM_Cubic, leaveTangent=-1.1642, time=0.1492, value=0.5}, <at 0.6397 0.096>]}
         }})
       1=&ColorFromCurve(ColorCurve=DataInterface() {propertyValues {
          AlphaCurve {keys [<at 0 1 RCIM_Cubic>]}
          BlueCurve {keys [<at 0.0118 0.1453 RCIM_Cubic>, {interpMode=RCIM_Cubic, leaveTangent=0.0305, time=0.6474, value=0.3729}]}
          GreenCurve {keys [<at 0.0118 1 RCIM_Cubic>, {interpMode=RCIM_Cubic, leaveTangent=-0.9856, time=0.6474, value=0.2809}]}
          RedCurve {keys [<at 0.0118 0.6025 RCIM_Cubic>, {interpMode=RCIM_Cubic, leaveTangent=0.4399, time=0.6474, value=0.9219}]}
         }})
       2=&ColorFromCurve(ColorCurve=DataInterface() {propertyValues {
          AlphaCurve {keys [<at 0 1 RCIM_Cubic>]}
          BlueCurve {keys [<at 0.0882 0 RCIM_Cubic>,<at 0.4302 0 RCIM_Cubic>,<at 0.7283 0>]}
          GreenCurve {keys [<at 0.0882 1.1798 RCIM_Cubic>, {interpMode=RCIM_Cubic, leaveTangent=-1.5694, time=0.4302, value=0.7945}, <at 0.7283 0.4217>]}
          RedCurve {keys [<at 0.0882 2 RCIM_Cubic>, {interpMode=RCIM_Cubic, leaveTangent=-2.07, time=0.4302, value=3}, <at 0.7283 1>]}
         }})
       3=&ColorFromCurve(ColorCurve=DataInterface() {propertyValues {
          AlphaCurve {keys [<at 0 1 RCIM_Cubic>]}
          BlueCurve {keys [<at 0 1 RCIM_Cubic>,<at 0.8159 1 RCIM_Cubic>]}
          GreenCurve {keys [<at 0 0.824 RCIM_Cubic>,<at 0.8159 0.8741 RCIM_Cubic>]}
          RedCurve {keys [<at 0 0 RCIM_Cubic>,<at 0.8159 0.5521 RCIM_Cubic>]}
         }})
       4=&ColorFromCurve(ColorCurve=DataInterface() {propertyValues {
          AlphaCurve {keys [<at 0 1 RCIM_Cubic>]}
          BlueCurve {keys [<at 0.1557 0.2952 RCIM_Cubic>,<at 0.4791 1>,<at 0.7906 1>,<at 1 0.6406>]}
          GreenCurve {keys [<at 0.1557 1 RCIM_Cubic>,<at 0.4791 0.8874>,<at 0.7906 0.3229>,<at 1 0.197>]}
          RedCurve {keys [<at 0.1557 0.2708 RCIM_Cubic>,<at 0.4791 0.3542>,<at 0.7906 0.46>,<at 1 0.1935>]}
         }})
       5=&ColorFromCurve(ColorCurve=DataInterface() {propertyValues {
          AlphaCurve {keys [<at 0 1 RCIM_Cubic>]}
          BlueCurve {keys [<at 0.0882 0.7046 RCIM_Cubic>,<at 0.8159 1 RCIM_Cubic>]}
          GreenCurve {keys [<at 0.0882 1 RCIM_Cubic>,<at 0.8159 0.107 RCIM_Cubic>]}
          RedCurve {keys [<at 0.0882 0 RCIM_Cubic>,<at 0.8159 0.3012 RCIM_Cubic>]}
         }})
       Index=$System.Variant
       Num=::NewEnumerator4
      ))
     B=&MakeVectorFromLinearColorRGB(Color=&ColorFromCurve(
       ColorCurve=DataInterface() {propertyValues {
         AlphaCurve {keys [<at 0 1>,<at 0.125 1>,<at 0.25 1>,<at 0.375 1>,<at 0.5 1>,<at 0.625 1>,<at 0.75 1>,<at 0.875 1>,<at 1 1>]}
         BlueCurve {keys [<at 0.104 0.849>,<at 0.1948 0.6172>,<at 0.25 0.6172>,<at 0.291 0.6939>,<at 0.333 0.7379>,<at 0.375 0.7835>,<at 0.416 0.7991>,<at 0.458 0.8148>,<at 0.5 0.8308>,<at 0.541 0.8148>,<at 0.583 0.7991>,<at 0.625 0.7991>,<at 0.666 0.7529>,<at 0.708 0.7084>,<at 0.75 0.5711>,<at 0.791 0.4735>,<at 0.8699 0.849>]}
         GreenCurve {keys [<at 0.104 0.6264>,<at 0.1948 0.5333>,<at 0.25 0.6376>,<at 0.291 0.7529>,<at 0.333 0.7835>,<at 0.375 0.8148>,<at 0.416 0.7991>,<at 0.458 0.8148>,<at 0.5 0.8469>,<at 0.541 0.7991>,<at 0.583 0.7991>,<at 0.625 0.8148>,<at 0.666 0.7991>,<at 0.708 0.7529>,<at 0.75 0.5711>,<at 0.791 0.5029>,<at 0.8699 0.6264>]}
         RedCurve {keys [<at 0.104 0.4318>,<at 0.1948 0.7084>,<at 0.25 0.8148>,<at 0.291 0.8879>,<at 0.333 0.8963>,<at 0.375 0.9131>,<at 0.416 0.8714>,<at 0.458 0.8714>,<at 0.5 0.9131>,<at 0.541 0.8632>,<at 0.583 0.8714>,<at 0.625 0.9131>,<at 0.666 0.8963>,<at 0.708 0.8714>,<at 0.75 0.7529>,<at 0.791 0.7529>,<at 0.8699 0.4318>]}
        }}
       CurveIndex=Linked() {linkedVariable {name=NPC.FortniteNPC.TimeOfDay_0-1, type=NiagaraFloat}}
      ))
    )
    ScaleA=true
    ScaleRGB=true
   )
   &DynamicMaterialParameters(
    Index 0 Param 1=&Add_Float(
     A=&RandomRangeFloat(Minimum=0, Maximum=1)
     B=&FloatFromCurve(
      FloatCurve=DataInterface() {propertyValues {Curve {keys [
          <at 0 0 RCIM_Cubic>
          {arriveTangent=1.5947, interpMode=RCIM_Cubic, leaveTangent=1.5947, tangentMode=RCTM_User, time=0.2557, value=0.5257}
          <at 1 1 RCIM_Cubic>
         ]}}}
      CurveIndex=$Particles.Age
      Scale Curve=0.25
     )
    )
    Module.Param0WriteEnabled=true
    Module.Param1WriteEnabled=true
    Module.Param2WriteEnabled=true
    Module.Param3WriteEnabled=true
   ) {module=DynamicMaterialParameters001}
   &SolveForcesAndVelocity(Clamp Velocity=false, Limit Acceleration=false)
  ]
  rendererClasses [&NiagaraSpriteRendererProperties]
  renderers [
   {rendererClass=&NiagaraSpriteRendererProperties}
   {rendererClass=&NiagaraSpriteRendererProperties, rendererIndex=1}
  ]
 }
 LeekPieces {
  emitterUpdateScript [
   &EmitterState(
    Enable Distance Culling=true
    Max Distance Response=::KillImmediately
    Module.bApplySpawnCountScaleByDistance=true
    Module.bUseMaxDistance=true
    Module.bUseMinDistance=false
    Scalability Mode=::NewEnumerator1
    Scale Spawn Count=true
    Spawn Count Scale By Distance=DataInterface() {propertyValues {Curve {keys [<at 0.3676 0.9657>,<at 0.9962 0.4982>]}}}
   )
   &SpawnBurst_Instantaneous(Spawn Count=&UniformRangedInt(Minimum=8, Maximum=15), Spawn Time=0.02, Use Spawn Probability=false)
  ]
  particleSpawnScript [
   /SetVariables(Particles.isRingLeek:NiagaraBool=&SetBoolByFloatComparison(A=&RandomRangeFloat(Minimum=0, Maximum=1), B=0.5)) {isSetParameters=true}
   &InitializeParticle(
    Color=&Multiply_LinearColorByFloat(
     Linear Color=&MultiplyLinearColor(
      Linear Color=&MakeCustomLinearColorFromBool(
       Bool=$Particles.isRingLeek
       True Color=&MergeHelper_ColorSwitch(
        0=&RandomRangeLinearColor(Minimum=<rgba 0.7247 1 0 1>, Maximum=<rgba 0.1689 0.599 0.0663 1>)
        1=&RandomRangeLinearColor(Minimum=<rgba 1 0.0916 0.1513 1>, Maximum=<rgba 0.3472 1 0.2292 1>)
        2=&RandomRangeLinearColor(Minimum=<rgba 2 1.3284 0.2166 1>, Maximum=<rgba 1 0.4594 0.2061 1>)
        3=&RandomRangeLinearColor(Minimum=<rgba 0 0.8283 1 1>, Maximum=<rgba 0.0335 0.5098 1 1>)
        4=&RandomRangeLinearColor(Minimum=<rgba 0.2248 0.2917 0.2421 1>, Maximum=<rgba 0.2003 0.1734 0.2448 1>)
        5=&RandomRangeLinearColor(Minimum=<rgba 0.2892 0.1957 0.4583 1>, Maximum=<rgba 0 0.536 1 1>)
        Index=$System.Variant
        Num=::NewEnumerator4
       )
       False Color=<rgba 0.5 0.5 0.5 1>
      )
      Scale Factor=&MakeVector4_FromLinearColor(LinearColor=&LerpLinearColors(
        StartColor=&ColorFromCurve(
         ColorCurve=DataInterface() {propertyValues {
           AlphaCurve {keys [<at 0 1>,<at 0.125 1>,<at 0.25 1>,<at 0.375 1>,<at 0.5 1>,<at 0.625 1>,<at 0.75 1>,<at 0.875 1>,<at 1 1>]}
           BlueCurve {keys [<at 0.0888 0.7448>,<at 0.1948 0.6172>,<at 0.25 0.6172>,<at 0.291 0.6939>,<at 0.333 0.7379>,<at 0.375 0.7835>,<at 0.416 0.7991>,<at 0.458 0.8148>,<at 0.5 0.8308>,<at 0.541 0.8148>,<at 0.583 0.7991>,<at 0.625 0.7991>,<at 0.666 0.7529>,<at 0.708 0.7084>,<at 0.75 0.5711>,<at 0.791 0.4735>,<at 0.8664 0.7448>]}
           GreenCurve {keys [<at 0.0888 0.5215>,<at 0.1948 0.5333>,<at 0.25 0.6376>,<at 0.291 0.7529>,<at 0.333 0.7835>,<at 0.375 0.8148>,<at 0.416 0.7991>,<at 0.458 0.8148>,<at 0.5 0.8469>,<at 0.541 0.7991>,<at 0.583 0.7991>,<at 0.625 0.8148>,<at 0.666 0.7991>,<at 0.708 0.7529>,<at 0.75 0.5711>,<at 0.791 0.5029>,<at 0.8664 0.5215>]}
           RedCurve {keys [<at 0.0888 0.2555>,<at 0.1948 0.7084>,<at 0.25 0.8148>,<at 0.291 0.8879>,<at 0.333 0.8963>,<at 0.375 0.9131>,<at 0.416 0.8714>,<at 0.458 0.8714>,<at 0.5 0.9131>,<at 0.541 0.8632>,<at 0.583 0.8714>,<at 0.625 0.9131>,<at 0.666 0.8963>,<at 0.708 0.8714>,<at 0.75 0.7529>,<at 0.791 0.7529>,<at 0.8664 0.2555>]}
          }}
         CurveIndex=Linked() {linkedVariable {name=NPC.FortniteNPC.TimeOfDay_0-1, type=NiagaraFloat}}
        )
        EndColor=<rgba 1 1 1 1>
        LerpFactor=&MergeHelper_FloatSwitch(0=0, 1=0, 2=0.6, 3=0, 4=0.7, 5=0.7, Index=$System.Variant, Num=::NewEnumerator4)
       ))
     )
     Float=&MakeCustomFloatFromBool(
      Bool=&SetBoolByFloatComparison(A=&RandomRangeFloat(Minimum=0, Maximum=1), B=0.8)
      True Float=2
      False Float=1
     )
    )
    Lifetime Max=0.8
    Lifetime Min=0.3
    Lifetime Mode=::NewEnumerator1
    Material Random:NiagaraFloat=&Add_Float(
     A=&MergeHelper_FloatSwitch(0=0.0288, 1=0.2024, 2=0.3616, 3=0.5288, 4=0.7312, 5=0.868, Index=$System.Variant, Num=::NewEnumerator4)
     B=&RandomRangeFloat(Minimum=-0.04, Maximum=0.04)
    )
    Mesh Renderer Array Visibility Mode=::NewEnumerator1
    Mesh Renderer Info=DataInterface() {propertyValues {MeshRenderer=&self:LeekPieces_5.NiagaraMeshRendererProperties_0}}
    Mesh Renderer Mesh Index:NiagaraInt32=&MakeCustomIntFromBool(Bool=$Particles.isRingLeek, True INT=3, False INT=&RandomRangeInt(Minimum=0, Maximum=2))
    Mesh Scale=&Multiply_VectorByFloat(
     Vector=&MakeCustomVectorFromBool(
      Bool=$Particles.isRingLeek
      True Vector=&RandomRangeVector(Minimum=<0.5 0.5 0.5>, Maximum=<0.8 0.8 1>)
      False Vector=&RandomRangeVector(Minimum=<0.3 0.5 0.5>, Maximum=<1 0.2 0.2>)
     )
     Float=&MakeCustomFloatFromBool(Bool=$User.bIsFrontEnd, True Float=$User.FrontEndScale, False Float=1)
    )
    Mesh Scale Mode=::NewEnumerator4
    UsePositionOffset=false
    Write Material Random=true
   )
   &ShapeLocation(MaskSpawnGroup=false, Override Seed=false, Sphere Radius=9, Sphere Surface Distribution=1, UseOwnerScale=true)
   &AddVelocity(
    Module.Use Falloff Distance=false
    Origin Offset=<0 0 -5>
    Override Seed=false
    Velocity Mode=::NewEnumerator1
    Velocity Speed=&Multiply_Float(
     A=&RandomRangeFloat(Minimum=10, Maximum=300)
     B=&MakeCustomFloatFromBool(Bool=$User.bIsFrontEnd, True Float=$User.FrontEndScale, False Float=1)
    )
   )
   &InitialMeshOrientation(Use Rotation Vector=false)
  ]
  particleUpdateScript [
   &ParticleState()
   &ScaleMeshSize(Scale Factor=&VectorFromFloat(Value=&FloatFromCurve(FloatCurve=DataInterface() {propertyValues {Curve {keys [<at 0 0>,<at 0.1 1>,<at 0.7 1>,<at 1 0>]}}})))
   &UpdateMeshOrientation(Rotation Vector=&RandomRangeVector(Minimum=<0 0 0>, Maximum=<1 1 1>))
   &GravityForce(Gravity=<0 0 -666>)
   &SolveForcesAndVelocity(Clamp Velocity=false, Limit Acceleration=false)
  ]
  rendererClasses [&NiagaraMeshRendererProperties]
  renderers [{rendererClass=&NiagaraMeshRendererProperties, rendererIndex=0}]
 }
 Sparks_Toon {
  emitterSpawnScript [
   /SetVariables(Emitter.UseSparkles:NiagaraBool=&SelectBoolFromArray(
     Array Sampling Mode:ENiagara_LifetimeMode=::NewEnumerator0
     Direct Array Index:NiagaraInt32=$System.Variant
     Direct Array Mode:ENiagara_WrapClamp=::NewEnumerator1
     Evaluation Type:ENiagaraRandomnessEvaluation=::NewEnumerator0
     Recalculate Random Each Loop:NiagaraBool=false
     Randomness Mode:ENiagaraRandomnessMode=::NewEnumerator0
     Random Seed:NiagaraInt32=?
     Fixed Random Seed:NiagaraBool=false
     Bool Selection Array:NiagaraDataInterfaceArrayBool=DataInterface() {propertyValues {GpuSyncMode=SyncCpuToGpu, MaxElements=0, BoolData [false,false,true,false,false,false]}}
    )) {isSetParameters=true}
   &self:Activate_Emitter(Active:NiagaraBool=$Emitter.UseSparkles) {module=Activate_Emitter}
  ]
  emitterUpdateScript [
   &EmitterState(Life Cycle Mode=::NewEnumerator1, Loop Behavior=::NewEnumerator1, Loop Delay=0.005, Loop Duration=0.25, UseLoopDelay=false)
   &SpawnRate(
    SpawnRate=&MakeCustomFloatFromBool(Bool=$Emitter.UseSparkles, True Float=10, False Float=0)
    Use Spawn Probability=false
   )
   &SpawnBurst_Instantaneous(
    Spawn Count=&MakeCustomIntFromBool(Bool=$Emitter.UseSparkles, True INT=5, False INT=0)
    Use Spawn Probability=false
    UseLoopCountLimit=false
   )
   &SpawnBurst_Instantaneous(
    Spawn Count=&MakeCustomIntFromBool(Bool=$Emitter.UseSparkles, True INT=&RandomRangeInt(Minimum=5, Maximum=10), False INT=0)
    Spawn Group=1
    Use Spawn Probability=false
    UseLoopCountLimit=false
   ) {module=SpawnBurst_Instantaneous001}
  ]
  particleSpawnScript [
   /SetVariables(Particles.SpawnGroup=$Emitter.SpawnGroup) {isSetParameters=true}
   &InitializeParticle(
    Color=<rgba 0.05 0.05 0.05 0.5>
    Lifetime Max=0.6
    Lifetime Min=0.4
    Lifetime Mode=::NewEnumerator1
    Sprite Size Mode=::NewEnumerator3
    Uniform Sprite Size=&Multiply_Float(
     A=&RandomRangeFloat(Minimum=10, Maximum=15)
     B=&FN_FrontEndFloatModifier(Gameplay=1, Front End=$User.FrontEndScale)
    )
    UsePositionOffset=false
    Write Material Random=true
   )
   /SetVariables(Particles.Position=&MergeHelper_VectorSwitch(
     0=Linked() {linkedVariable {name=StackContext.Position, type=NiagaraPosition}}
     1=Linked() {linkedVariable {name=StackContext.Position, type=NiagaraPosition}}
     2=$Output.SampleParticlesFromOtherEmitter.SampledPosition
     3=?
     Index=$Particles.SpawnGroup
     Num=::NewEnumerator1
    )) {isSetParameters=true}
   &SphereLocation(
    MaskSpawnGroup=false
    Override Seed=false
    Sphere Radius=&Multiply_Float(
     A=&RandomRangeFloat(Minimum=2, Maximum=25)
     B=&FN_FrontEndFloatModifier(Gameplay=1, Front End=$User.FrontEndScale)
    )
    Surface Expansion Mode=::NewEnumerator1
    Surface Only=true
    UseNonUniformScale=false
    UseOffset=true
    UseSphereOrientation=false
   )
   &CameraOffset(Camera Offset Amount=1)
   /SetVariables(Particles.VisibilityTag=$Particles.SpawnGroup) {isSetParameters=true}
   &AddVelocity(
    Module.Use Falloff Distance=false
    Override Seed=false
    Velocity Mode=::NewEnumerator1
    Velocity Speed=&MergeHelper_FloatSwitch(
     0=0
     1=&Multiply_Float(
      A=&RandomRangeFloat(Minimum=10, Maximum=400)
      B=&Multiply_Float(A=0.58, B=&FN_FrontEndFloatModifier(Gameplay=1, Front End=$User.FrontEndScale))
     )
     2=?
     3=?
     Index=$Particles.SpawnGroup
     Num=::NewEnumerator0
    )
   )
   /SetVariables(Particles.SpriteSize=&Multiply_Vector2D_ByFloat(
     Float=&MergeHelper_FloatSwitch(0=1, 1=0.5, 2=?, 3=?, Index=$Particles.SpawnGroup, Num=::NewEnumerator0)
     Vector2D=$StackContext.SpriteSize
    )) {isSetParameters=true}
   &AddVelocity(
    Velocity=&MergeHelper_VectorSwitch(
     0=$Output.SphereLocation.SphereNormal
     1=$Output.SphereLocation.SphereNormal
     2=$Output.SphereLocation.SphereNormal
     3=?
     Index=$Particles.SpawnGroup
     Num=::NewEnumerator1
    )
    Velocity Speed Scale=&MergeHelper_FloatSwitch(
     0=&RandomRangeFloat(Minimum=1, Maximum=50)
     1=1
     2=33
     3=?
     Index=$Particles.SpawnGroup
     Num=::NewEnumerator1
    )
   ) {module=AddVelocity002}
  ]
  particleUpdateScript [
   &ParticleState()
   &SolveForcesAndVelocity(Clamp Velocity=false, Limit Acceleration=false)
   &Color(
    Color=&MergeHelper_ColorSwitch(
     0=&ColorFromCurve(ColorCurve=DataInterface() {propertyValues {
        AlphaCurve {keys [<at 0.0065 0>,<at 0.3936 0.5 RCIM_Cubic>,<at 1 0 RCIM_Cubic>]}
        BlueCurve {keys [<at 0 0 RCIM_Cubic>,<at 0.7105 0 RCIM_Cubic>]}
        GreenCurve {keys [<at 0 0.75 RCIM_Cubic>,<at 0.7105 0.0368 RCIM_Cubic>]}
        RedCurve {keys [<at 0 3 RCIM_Cubic>,<at 0.7105 0.3594 RCIM_Cubic>]}
       }})
     1=<rgba 0.1 0.0495 0.015 1>
     2=?
     3=?
     Index=$Particles.SpawnGroup
     Num=::NewEnumerator0
    )
    Scale Alpha=&FloatFromCurve(
     FloatCurve=DataInterface() {propertyValues {Curve {keys [
         {interpMode=RCIM_Cubic, tangentMode=RCTM_User, time=-0.0063, value=0.0184}
         {arriveTangent=-0.8483, interpMode=RCIM_Cubic, leaveTangent=-0.8483, tangentMode=RCTM_User, time=0.1718, value=1.7147}
         {arriveTangent=-5.0978, interpMode=RCIM_Cubic, leaveTangent=-5.0978, tangentMode=RCTM_User, time=0.2818, value=1.2143}
         {arriveTangent=-0.2783, interpMode=RCIM_Cubic, leaveTangent=-0.2783, tangentMode=RCTM_User, time=0.4, value=0.7}
         {arriveTangent=0.3112, interpMode=RCIM_Cubic, leaveTangent=0.3112, tangentMode=RCTM_User, time=0.5407, value=1.3419}
         {arriveTangent=-3.7119, interpMode=RCIM_Cubic, leaveTangent=-3.7119, tangentMode=RCTM_User, time=0.7242, value=0.3749}
         {interpMode=RCIM_Cubic, tangentMode=RCTM_User, time=1.0011, value=0.002}
        ]}}}
     Scale Curve=0.8
    )
    Scale Color=&VectorFromFloat(Value=50)
   )
   &ScaleSpriteSize(Scale Factor=&Vector2DFromFloat(Value=&FloatFromCurve(
      FloatCurve=DataInterface() {propertyValues {Curve {keys [
          {arriveTangent=4.1302, interpMode=RCIM_Cubic, leaveTangent=4.1302, tangentMode=RCTM_User, time=0, value=0}
          {interpMode=RCIM_Cubic, tangentMode=RCTM_User, time=0.5, value=1}
          {tangentMode=RCTM_User, time=1, value=0}
         ]}}}
      Scale Curve=&Multiply_Float(
       A=&MaskFloatBySpawnGroup(Masked Float=1, Passthrough Float=1.5)
       B=&FN_FrontEndFloatModifier(Gameplay=1, Front End=0.5)
      )
     )))
   &Drag(
    Drag=&MergeHelper_FloatSwitch(0=0, 1=6, 2=?, 3=?, Index=$Particles.SpawnGroup, Num=::NewEnumerator0)
    Use Linear Drag=true
    Use Rotational Drag=true
   )
   &SolveForcesAndVelocity(Clamp Velocity=false, Limit Acceleration=false) {module=SolveForcesAndVelocity001}
  ]
  rendererClasses [&NiagaraSpriteRendererProperties]
  renderers [
   {rendererClass=&NiagaraSpriteRendererProperties}
   {rendererClass=&NiagaraSpriteRendererProperties, rendererIndex=1}
   {rendererClass=&NiagaraSpriteRendererProperties, rendererIndex=2}
  ]
 }
 Sprite_Ice_Flakes {
  emitterSpawnScript [
   /SetVariables(
    Emitter.SpawnParticles:NiagaraBool=&SelectBoolFromArray(
     Array Sampling Mode:ENiagara_LifetimeMode=::NewEnumerator0
     Direct Array Index:NiagaraInt32=$System.Variant
     Direct Array Mode:ENiagara_WrapClamp=::NewEnumerator1
     Evaluation Type:ENiagaraRandomnessEvaluation=::NewEnumerator0
     Recalculate Random Each Loop:NiagaraBool=false
     Randomness Mode:ENiagaraRandomnessMode=::NewEnumerator0
     Random Seed:NiagaraInt32=?
     Fixed Random Seed:NiagaraBool=false
     Bool Selection Array:NiagaraDataInterfaceArrayBool=DataInterface() {propertyValues {GpuSyncMode=SyncCpuToGpu, MaxElements=0, BoolData [false,false,false,true,true,true]}}
    )
    Emitter.isGalaxy:NiagaraBool=&SelectBoolFromArray(
     Array Sampling Mode:ENiagara_LifetimeMode=::NewEnumerator0
     Direct Array Index:NiagaraInt32=$System.Variant
     Direct Array Mode:ENiagara_WrapClamp=::NewEnumerator1
     Evaluation Type:ENiagaraRandomnessEvaluation=::NewEnumerator0
     Recalculate Random Each Loop:NiagaraBool=false
     Randomness Mode:ENiagaraRandomnessMode=::NewEnumerator0
     Random Seed:NiagaraInt32=?
     Fixed Random Seed:NiagaraBool=false
     Bool Selection Array:NiagaraDataInterfaceArrayBool=DataInterface() {propertyValues {GpuSyncMode=SyncCpuToGpu, MaxElements=0, BoolData [false,false,false,false,false,true]}}
    )
   ) {isSetParameters=true}
   &self:Activate_Emitter(Active:NiagaraBool=$Emitter.SpawnParticles) {module=Activate_Emitter}
  ]
  emitterUpdateScript [
   &EmitterState(Life Cycle Mode=::NewEnumerator1, Loop Behavior=::NewEnumerator1, Loop Duration=0.85, UseLoopDelay=false)
   &SpawnBurst_Instantaneous(
    Spawn Count=&MakeCustomIntFromBool(Bool=$Emitter.SpawnParticles, True INT=&RandomRangeInt(Minimum=10, Maximum=15), False INT=0)
    Use Spawn Probability=false
    UseLoopCountLimit=false
   )
  ]
  particleSpawnScript [
   &InitializeParticle(
    Color=<rgba 1.64 2.928 5 1>
    Lifetime Max=0.8
    Lifetime Min=0.3
    Lifetime Mode=::NewEnumerator1
    Ribbon Width=0.5
    Ribbon Width Mode=::NewEnumerator1
    Sprite Size Mode=::NewEnumerator3
    Uniform Sprite Size=&Multiply_Float(
     A=&Multiply_Float(
      A=&RandomRangeFloat(Minimum=2, Maximum=5)
      B=&MergeHelper_FloatSwitch(
       0=0
       1=0
       2=0
       3=1
       4=&RandomRangeFloat(Minimum=3, Maximum=5)
       5=&RandomRangeFloat(Minimum=3, Maximum=5)
       Index=$System.Variant
       Num=::NewEnumerator4
      )
     )
     B=&FN_FrontEndFloatModifier(Gameplay=1, Front End=$User.FrontEndScale)
    )
    UsePositionOffset=true
    Write Material Random=true
   )
   &ShapeLocation(
    MaskSpawnGroup=false
    Override Seed=false
    Sphere Radius=&Multiply_Float(A=10, B=&FN_FrontEndFloatModifier(Gameplay=1, Front End=$System.FrontendScale))
    Transform Order:ENiagara_TransformOrder=::NewEnumerator1
    UseOwnerScale=true
   ) {module=ShapeLocation001}
   &AddVelocity(
    Module.Use Falloff Distance=false
    Override Seed=false
    Velocity Mode=::NewEnumerator1
    Velocity Speed=&Multiply_Float(
     A=&RandomRangeFloat(Minimum=100, Maximum=600)
     B=&FN_FrontEndFloatModifier(Gameplay=1, Front End=$System.FrontendScale)
    )
   ) {module=AddVelocity001}
   /SetVariables(Particles.VisibilityTag=&MergeHelper_IntSwitch(0=0, 1=0, 2=0, 3=0, 4=1, 5=1, Index=$System.Variant, Num=::NewEnumerator4)) {isSetParameters=true}
   /SetVariables(Particles.RibbonID=/Niagara/DynamicInputs/NiagaraID/MakeNiagaraIDFromInt(
     Index:NiagaraInt32=/Niagara/DynamicInputs/Divide/Divide_Int(A:NiagaraInt32=&ReturnParticleIDAttributes(), B:NiagaraInt32=3)
     Acquire Tag:NiagaraInt32=0
    )) {isSetParameters=true}
  ]
  particleUpdateScript [
   &ParticleState()
   &AccelerationForce(Acceleration=&Multiply_VectorByFloat(Vector=<0 0 -100>, Float=&FN_FrontEndFloatModifier(Gameplay=1, Front End=$System.FrontendScale)))
   &Drag(Drag=10, Use Linear Drag=true, Use Rotational Drag=true)
   &CurlNoiseForce(Noise Strength=200, Pan Noise=false)
   &SolveForcesAndVelocity(Clamp Velocity=false, Limit Acceleration=false)
   &ScaleSpriteSize(
    Uniform Curve Scale=&FN_FrontEndFloatModifier(Gameplay=1, Front End=$System.FrontendScale)
    Uniform Curve Sprite Scale=DataInterface() {propertyValues {Curve {keys [
        {arriveTangent=6.9483, interpMode=RCIM_Cubic, leaveTangent=6.9483, tangentMode=RCTM_User, time=0, value=0}
        {interpMode=RCIM_Cubic, tangentMode=RCTM_User, time=0.5, value=1}
        {tangentMode=RCTM_User, time=1, value=0}
       ]}}}
   )
   &ScaleColor(
    Scale RGB=&MakeVectorFromLinearColorRGB(Color=&ColorFromCurve(
      ColorCurve=DataInterface() {propertyValues {
        AlphaCurve {keys [<at 0 1 RCIM_Cubic>]}
        BlueCurve {keys [
          <at 0.0932 0.3744>
          {arriveTangent=1.0386, interpMode=RCIM_Cubic, time=0.3498, value=1}
          {arriveTangent=0.7495, interpMode=RCIM_Cubic, leaveTangent=0.7495, time=0.6308, value=1}
          <at 0.9279 1>
         ]}
        GreenCurve {keys [
          <at 0.0932 1>
          {arriveTangent=-1.0894, interpMode=RCIM_Cubic, time=0.3498, value=0.7632}
          {arriveTangent=-0.8299, interpMode=RCIM_Cubic, leaveTangent=-0.8299, time=0.6308, value=0.3438}
          <at 0.9279 0.3073>
         ]}
        RedCurve {keys [
          <at 0.0932 0.2598>
          {arriveTangent=1.08, interpMode=RCIM_Cubic, time=0.3498, value=0.276}
          {arriveTangent=0.2625, interpMode=RCIM_Cubic, leaveTangent=0.2625, time=0.6308, value=0.9103}
          <at 0.9279 0.4788>
         ]}
       }}
      CurveIndex=&RandomRangeFloat(Minimum=0, Maximum=1)
     ))
    ScaleA=true
    ScaleRGB=true
   ) {enabled=false}
   /SetVariables(
    Particles.GlowColor=&MakeLinearColorFromVectorAndFloat(
     "Vector (RGB)"=&MakeVectorFromLinearColorRGB(Color=&MergeHelper_ColorSwitch(
       0=&Multiply_LinearColorByFloat(
        Linear Color=&ColorFromCurve(
         ColorCurve=DataInterface() {propertyValues {
           AlphaCurve {keys [<at 0 1 RCIM_Cubic>]}
           BlueCurve {keys [
             <at 0 0.3101>
             {arriveTangent=0.3864, interpMode=RCIM_Cubic, time=0.1915, value=0.5}
             <at 0.5198 0.5>
             {arriveTangent=-0.3954, interpMode=RCIM_Cubic, leaveTangent=-0.3954, time=0.7638, value=0.5}
             <at 1 0.3101>
            ]}
           GreenCurve {keys [
             <at 0 0.5677>
             {arriveTangent=-0.8426, interpMode=RCIM_Cubic, time=0.1915, value=0.3594}
             <at 0.5198 0.1536>
             {arriveTangent=0.8622, interpMode=RCIM_Cubic, leaveTangent=0.8622, time=0.7638, value=0.0833}
             <at 1 0.5677>
            ]}
           RedCurve {keys [
             <at 0 0.1117>
             {arriveTangent=0.2598, interpMode=RCIM_Cubic, time=0.1915, value=0.0703}
             <at 0.5198 0.2394>
             {arriveTangent=-0.2659, interpMode=RCIM_Cubic, leaveTangent=-0.2659, time=0.7638, value=0.4431}
             <at 1 0.1117>
            ]}
          }}
         CurveIndex=&Frac_Float(Float=&Add_Float(A=$Particles.NormalizedAge, B=&RandomRangeFloat(Minimum=0, Maximum=-0.5)))
        )
        Float=&RandomRangeFloat(Minimum=0.2, Maximum=1)
       )
       1=&Multiply_LinearColorByFloat(
        Linear Color=&ColorFromCurve(
         ColorCurve=DataInterface() {propertyValues {
           AlphaCurve {keys [<at 0 1 RCIM_Cubic>]}
           BlueCurve {keys [
             <at 0.0284 0.8677>
             {arriveTangent=-0.4812, interpMode=RCIM_Cubic, leaveTangent=-0.4812, time=0.4737, value=0.8646}
             <at 0.7925 0.5>
             <at 0.9918 0.8646>
            ]}
           GreenCurve {keys [
             <at 0.0284 1>
             {arriveTangent=-1.2221, interpMode=RCIM_Cubic, leaveTangent=-1.2221, time=0.4737, value=0.1067}
             <at 0.7925 0.0662>
             <at 0.9918 0.7183>
            ]}
           RedCurve {keys [
             <at 0.0284 0>
             {arriveTangent=0.4161, interpMode=RCIM_Cubic, leaveTangent=0.4161, time=0.4737, value=0.231}
             <at 0.7925 0.3179>
             <at 0.9918 0>
            ]}
          }}
         CurveIndex=&Frac_Float(Float=&Add_Float(A=$Particles.NormalizedAge, B=&RandomRangeFloat(Minimum=0, Maximum=1)))
        )
        Float=&RandomRangeFloat(Minimum=0.5, Maximum=0.5)
       )
       2=?
       3=?
       Index=/Niagara/DynamicInputs/Subtract/Subtract_Integer(A:NiagaraInt32=$System.Variant, B:NiagaraInt32=4)
       Num=::NewEnumerator0
      ))
     "Float (Alpha)"=1
    )
    Particles.RibbonColor:LinearColor=&MakeLinearColorFromVectorAndFloat(
     "Vector (RGB)"=&MakeVectorFromLinearColorRGB(Color=&Multiply_LinearColorByFloat(
       Linear Color=&ColorFromCurve(
        ColorCurve=DataInterface() {propertyValues {
          AlphaCurve {keys [<at 0 1 RCIM_Cubic>]}
          BlueCurve {keys [
            <at 0.01 0.8677>
            {arriveTangent=-0.0032, interpMode=RCIM_Cubic, leaveTangent=-0.0032, time=0.2634, value=0.8646}
            <at 0.6801 0.5>
            <at 0.9918 0.8646>
           ]}
          GreenCurve {keys [
            <at 0.01 1>
            {arriveTangent=-0.287, interpMode=RCIM_Cubic, leaveTangent=-0.287, time=0.2634, value=0.1067}
            <at 0.6801 0.0662>
            <at 0.9918 0.7183>
           ]}
          RedCurve {keys [<at 0.01 0>,<at 0.2634 0.231 RCIM_Cubic>,<at 0.6801 0.3179>,<at 0.9918 0>]}
         }}
        CurveIndex=&Frac_Float(Float=&Add_Float(
          A=&Multiply_Float(A=$Particles.NormalizedAge, B=0.3)
          B=&RandomRangeFloat(Minimum=0, Maximum=1)
         ))
       )
       Float=&RandomRangeFloat(Minimum=0.2, Maximum=0.5)
      ))
     "Float (Alpha)"=&FloatFromCurve(FloatCurve=DataInterface() {propertyValues {Curve {keys [<at 0.5 1>,<at 1 0>]}}}, Scale Curve=1.1)
    )
   ) {isSetParameters=true}
  ]
  rendererClasses [&NiagaraSpriteRendererProperties, &NiagaraRibbonRendererProperties]
  renderers [
   {rendererClass=&NiagaraSpriteRendererProperties}
   {rendererClass=&NiagaraSpriteRendererProperties, rendererIndex=1}
   {rendererClass=&NiagaraRibbonRendererProperties, rendererIndex=2}
  ]
 }
}
summary {
 systemName=NS_Pickaxe_DashSurge_Impact
 emitters [
  {
   emitterName=Flash_Elements
   rendererClasses [&NiagaraSpriteRendererProperties, &NiagaraLightRendererProperties]
  }
  {emitterName=Impact_Smoke, rendererClasses [&NiagaraSpriteRendererProperties]}
  {emitterName=LeekPieces, rendererClasses [&NiagaraMeshRendererProperties]}
  {emitterName=Dark_Glow, rendererClasses [&NiagaraSpriteRendererProperties]}
  {emitterName=Sparks_Toon, rendererClasses [&NiagaraSpriteRendererProperties]}
  {
   emitterName=Sprite_Ice_Flakes
   rendererClasses [&NiagaraSpriteRendererProperties, &NiagaraRibbonRendererProperties]
  }
  {emitterName=Ice_Shards, rendererClasses [&NiagaraMeshRendererProperties, &NiagaraSpriteRendererProperties]}
 ]
 userVariables [
  User.bIsFrontEnd {type=NiagaraBool}
  User.bIsFrontEndPreview {type=NiagaraBool}
  User.FrontEndScale {type=NiagaraFloat, defaultValue=0.4, description=""}
  User.PhysicalSurface {type=NiagaraInt32}
  User.Variant {type=NiagaraFloat, defaultValue=0, description=""}
 ]
}
