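# Niagara swing effect for the DashSurge pickaxe, with look selected by User.Variant.
# Trail draws a short-lived anim-trail ribbon on the system's skeletal mesh, tinted by time-of-day color curves and faded by System.TrailFade.
# Variant_Sprites spawns drifting, curl-noise sparkles only for variants 2 and up, switching size, color and visibility tag per variant.
# Its sprite and ribbon renderers share particles, with ribbon IDs grouping them in threes and animated gradients supplying glow colors.
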NiagaraSystem /BRCosmetics/Weapons/FORT_Melee/Pickaxe_DashSurge/FX/NS_Pickaxe_DashSurge_Swing
dependencies {
 usedDataInterfaces [
  &NiagaraDataInterfaceColorCurve
  &NiagaraDataInterfaceCurve
  /Script/Niagara.NiagaraDataInterfaceArrayBool
  &NiagaraDataInterfaceSkeletalMesh
 ]
 usedDynamicInputs [
  &MergeHelper_VectorSwitch
  &self:SectionIndexFloatSelector
 ]
 usedModules [
  &self:SystemSpawnScript.SystemScriptSource.SystemScriptGraph.NiagaraNodeAssignment_3.SetVariables_#1
  &self:SystemSpawnScript.SystemScriptSource.SystemScriptGraph.NiagaraNodeAssignment_4.SetVariables_#2
  &SystemState
  &self:SystemSpawnScript.SystemScriptSource.SystemScriptGraph.NiagaraNodeAssignment_1.SetVariables_#3
  &self:SystemSpawnScript.SystemScriptSource.SystemScriptGraph.NiagaraNodeAssignment_2.SetVariables_#4
 ]
}
emitters {
 Trail {
  emitterSpawnScript [/SetVariables(Emitter.TOD Color:LinearColor=&LerpLinearColors(
     StartColor=&ColorFromCurve(
      ColorCurve=DataInterface() {propertyValues {
        AlphaCurve {keys [<at 0 1>,<at 0.125 1>,<at 0.25 1>,<at 0.375 1>,<at 0.5 1>,<at 0.625 1>,<at 0.75 1>,<at 0.875 1>,<at 1 1>]}
        BlueCurve {keys [<at 0.1156 0.4375>,<at 0.1948 0.6172>,<at 0.25 0.6172>,<at 0.291 0.6939>,<at 0.333 0.7379>,<at 0.375 0.7835>,<at 0.416 0.7991>,<at 0.458 0.8148>,<at 0.5 0.8308>,<at 0.541 0.8148>,<at 0.583 0.7991>,<at 0.625 0.7991>,<at 0.666 0.7529>,<at 0.708 0.7084>,<at 0.75 0.5711>,<at 0.791 0.4735>,<at 0.8705 0.4375>]}
        GreenCurve {keys [<at 0.1156 0.3399>,<at 0.1948 0.5333>,<at 0.25 0.6376>,<at 0.291 0.7529>,<at 0.333 0.7835>,<at 0.375 0.8148>,<at 0.416 0.7991>,<at 0.458 0.8148>,<at 0.5 0.8469>,<at 0.541 0.7991>,<at 0.583 0.7991>,<at 0.625 0.8148>,<at 0.666 0.7991>,<at 0.708 0.7529>,<at 0.75 0.5711>,<at 0.791 0.5029>,<at 0.8705 0.3399>]}
        RedCurve {keys [<at 0.1156 0.2301>,<at 0.1948 0.7084>,<at 0.25 0.8148>,<at 0.291 0.8879>,<at 0.333 0.8963>,<at 0.375 0.9131>,<at 0.416 0.8714>,<at 0.458 0.8714>,<at 0.5 0.9131>,<at 0.541 0.8632>,<at 0.583 0.8714>,<at 0.625 0.9131>,<at 0.666 0.8963>,<at 0.708 0.8714>,<at 0.75 0.7529>,<at 0.791 0.7529>,<at 0.8705 0.2301>]}
       }}
      CurveIndex=Linked() {linkedVariable {name=NPC.FortniteNPC.TimeOfDay_0-1, type=NiagaraFloat}}
     )
     EndColor=&ColorFromCurve(
      ColorCurve=DataInterface() {propertyValues {
        AlphaCurve {keys [<at 0 1>,<at 0.125 1>,<at 0.25 1>,<at 0.375 1>,<at 0.5 1>,<at 0.625 1>,<at 0.75 1>,<at 0.875 1>,<at 1 1>]}
        BlueCurve {keys [<at 0.1156 0.5625>,<at 0.1948 0.6172>,<at 0.25 0.6172>,<at 0.291 0.6939>,<at 0.333 0.7379>,<at 0.375 0.7835>,<at 0.416 0.7991>,<at 0.458 0.8148>,<at 0.5 0.8308>,<at 0.541 0.8148>,<at 0.583 0.7991>,<at 0.625 0.7991>,<at 0.666 0.7529>,<at 0.708 0.7084>,<at 0.75 0.5711>,<at 0.791 0.4735>,<at 0.881 0.5625>]}
        GreenCurve {keys [<at 0.1156 0.2834>,<at 0.1948 0.5333>,<at 0.25 0.6376>,<at 0.291 0.7529>,<at 0.333 0.7835>,<at 0.375 0.8148>,<at 0.416 0.7991>,<at 0.458 0.8148>,<at 0.5 0.8469>,<at 0.541 0.7991>,<at 0.583 0.7991>,<at 0.625 0.8148>,<at 0.666 0.7991>,<at 0.708 0.7529>,<at 0.75 0.5711>,<at 0.791 0.5029>,<at 0.881 0.2834>]}
        RedCurve {keys [<at 0.1156 0.1758>,<at 0.1948 0.7084>,<at 0.25 0.8148>,<at 0.291 0.8879>,<at 0.333 0.8963>,<at 0.375 0.9131>,<at 0.416 0.8714>,<at 0.458 0.8714>,<at 0.5 0.9131>,<at 0.541 0.8632>,<at 0.583 0.8714>,<at 0.625 0.9131>,<at 0.666 0.8963>,<at 0.708 0.8714>,<at 0.75 0.7529>,<at 0.791 0.7529>,<at 0.881 0.1758>]}
       }}
      CurveIndex=Linked() {linkedVariable {name=NPC.FortniteNPC.TimeOfDay_0-1, type=NiagaraFloat}}
     )
     LerpFactor=&MergeHelper_FloatSwitch(0=1, 1=1, 2=0, 3=1, 4=1, 5=0, Index=&MakeIntFromFloat(Float=$User.Variant), Num=::NewEnumerator4)
    )) {isSetParameters=true}]
  emitterUpdateScript [
   &EmitterState()
   &FN_SpawnAnimTrail_v2()
  ]
  particleSpawnScript [
   /Game/Effects/Niagara/Modules/AnimTrail/FN_InitializeAnimTrail/FN_InitializeAnimTrailBoneIndecies(First Bone or Socket Index=1)
   &FN_InitializeAnimTrail(
    Color=&MakeLinearColorFromVectorAndFloat(
     "Vector (RGB)"=&MakeVectorFromLinearColorRGB(Color=$Emitter.TOD Color)
     "Float (Alpha)"=&Multiply_Float(
      A=$System.TrailFade
      B=&MakeCustomFloatFromBool(
       Bool=&SetBoolByIntComparison(A=&MakeIntFromFloat(Float=$User.Variant), B=4, Comparison Type=::NewEnumerator2)
       True Float=2
       False Float=1
      )
     )
    )
    Lifetime=0.13
    SkeletalMesh=$System.Skeletal Mesh
   ) {module=FN_InitializeAnimTrail001}
   &AddVelocity(Velocity=<500 0 0>) {enabled=false}
   &DynamicMaterialParameters(Write Parameter Index 0=false)
  ]
  particleUpdateScript [
   &SolveForcesAndVelocity(Clamp Velocity=false, Limit Acceleration=false) {enabled=false}
   &ScaleColor(
    Scale Alpha=&FloatFromCurve(FloatCurve=DataInterface() {propertyValues {Curve {keys [<at 0 1 RCIM_Cubic>,<at 0.5 0 RCIM_Cubic>]}}})
    Scale RGB=&Multiply_Vector(
     A=<1 1 1>
     B=&MakeVectorFromLinearColorRGB(Color=&ColorFromCurve(
       ColorCurve=DataInterface() {propertyValues {
         AlphaCurve {keys [<at 0 1>,<at 0.125 1>,<at 0.25 1>,<at 0.375 1>,<at 0.5 1>,<at 0.625 1>,<at 0.75 1>,<at 0.875 1>,<at 1 1>]}
         BlueCurve {keys [<at 0.1156 0.624>,<at 0.1948 0.6172>,<at 0.25 0.6172>,<at 0.291 0.6939>,<at 0.333 0.7379>,<at 0.375 0.7835>,<at 0.416 0.7991>,<at 0.458 0.8148>,<at 0.5 0.8308>,<at 0.541 0.8148>,<at 0.583 0.7991>,<at 0.625 0.7991>,<at 0.666 0.7529>,<at 0.708 0.7084>,<at 0.75 0.5711>,<at 0.791 0.4735>,<at 0.9093 0.624>]}
         GreenCurve {keys [<at 0.1156 0.3763>,<at 0.1948 0.5333>,<at 0.25 0.6376>,<at 0.291 0.7529>,<at 0.333 0.7835>,<at 0.375 0.8148>,<at 0.416 0.7991>,<at 0.458 0.8148>,<at 0.5 0.8469>,<at 0.541 0.7991>,<at 0.583 0.7991>,<at 0.625 0.8148>,<at 0.666 0.7991>,<at 0.708 0.7529>,<at 0.75 0.5711>,<at 0.791 0.5029>,<at 0.9093 0.3763>]}
         RedCurve {keys [<at 0.1156 0.0976>,<at 0.1948 0.7084>,<at 0.25 0.8148>,<at 0.291 0.8879>,<at 0.333 0.8963>,<at 0.375 0.9131>,<at 0.416 0.8714>,<at 0.458 0.8714>,<at 0.5 0.9131>,<at 0.541 0.8632>,<at 0.583 0.8714>,<at 0.625 0.9131>,<at 0.666 0.8963>,<at 0.708 0.8714>,<at 0.75 0.7529>,<at 0.791 0.7529>,<at 0.9093 0.0976>]}
        }}
       CurveIndex=Linked() {linkedVariable {name=NPC.FortniteNPC.TimeOfDay_0-1, type=NiagaraFloat}}
      ))
    )
    ScaleA=true
    ScaleRGB=false
   ) {module=ScaleColor001}
   &FN_UpdateAnimTrail(SkeletalMesh=$System.Skeletal Mesh)
  ]
  rendererClasses [&NiagaraRibbonRendererProperties]
  renderers [{rendererClass=&NiagaraRibbonRendererProperties}]
 }
 Variant_Sprites {
  emitterSpawnScript [/SetVariables(
    Emitter.UseSparkles:NiagaraBool=/Niagara/DynamicInputs/Arrays/SelectBoolFromArray(
     Array Sampling Mode:ENiagara_LifetimeMode=::NewEnumerator0
     Direct Array Index:NiagaraInt32=&MakeIntFromFloat(Float=$User.Variant)
     Direct Array Mode:ENiagara_WrapClamp=::NewEnumerator1
     Evaluation Type:ENiagaraRandomnessEvaluation=::NewEnumerator0
     Recalculate Random Each Loop:NiagaraBool=false
     Randomness Mode:ENiagaraRandomnessMode=::NewEnumerator0
     Random Seed:NiagaraInt32=?
     Fixed Random Seed:NiagaraBool=false
     Bool Selection Array:NiagaraDataInterfaceArrayBool=DataInterface() {propertyValues {GpuSyncMode=SyncCpuToGpu, MaxElements=0, BoolData [false,false,true,true,true,true]}}
    )
    Emitter.isGalaxy:NiagaraBool=&SelectBoolFromArray(
     Array Sampling Mode:ENiagara_LifetimeMode=::NewEnumerator0
     Direct Array Index:NiagaraInt32=&MakeIntFromFloat(Float=$User.Variant)
     Direct Array Mode:ENiagara_WrapClamp=::NewEnumerator1
     Evaluation Type:ENiagaraRandomnessEvaluation=::NewEnumerator0
     Recalculate Random Each Loop:NiagaraBool=false
     Randomness Mode:ENiagaraRandomnessMode=::NewEnumerator0
     Random Seed:NiagaraInt32=?
     Fixed Random Seed:NiagaraBool=false
     Bool Selection Array:NiagaraDataInterfaceArrayBool=DataInterface() {propertyValues {GpuSyncMode=SyncCpuToGpu, MaxElements=0, BoolData [false,false,false,false,false,true]}}
    )
   ) {isSetParameters=true}]
  emitterUpdateScript [
   &EmitterState(
    Enable Distance Culling=true
    Enable Visibility Culling=true
    Life Cycle Mode=::NewEnumerator1
    Loop Behavior=::NewEnumerator1
    Loop Duration=0.5
    MaxDistance=2500
    Module.bApplySpawnCountScaleByDistance=true
    Module.bUseMaxDistance=true
    Module.bUseMinDistance=false
    Scalability Mode=::NewEnumerator1
    Scale Spawn Count=true
    UseLoopDelay=false
   )
   &SpawnRate(
    SpawnRate=&MakeCustomFloatFromBool(Bool=$Emitter.UseSparkles, True Float=100, False Float=0)
    Use Spawn Probability=false
   )
  ]
  particleSpawnScript [
   &InitializeParticle(
    Color=&MergeHelper_ColorSwitch(
     0=<rgba 0 0 0 1>
     1=<rgba 0 0 0 0>
     2=&RandomRangeLinearColor(Minimum=<rgba 10 3.6667 0.5 1>, Maximum=<rgba 10 5.5 1 1>)
     3=<rgba 1 1 1 1>
     4=<rgba 1 1 1 1>
     5=<rgba 1 1 1 1>
     Index=&MakeIntFromFloat(Float=$User.Variant)
     Num=::NewEnumerator4
    )
    Lifetime Max=0.4
    Lifetime Min=0.4
    Lifetime Mode=::NewEnumerator1
    Position Offset=&RandomRangeVector(Minimum=<-5 -5 60>, Maximum=<-5 5 10>)
    Position Offset Coordinate Space=::Local
    Ribbon Width=0.5
    Ribbon Width Mode=::NewEnumerator1
    Sprite Size Mode=::NewEnumerator3
    Uniform Sprite Size=&MergeHelper_FloatSwitch(
     0=0
     1=0
     2=&RandomRangeFloat(Minimum=2, Maximum=4)
     3=&RandomRangeFloat(Minimum=3, Maximum=5)
     4=&RandomRangeFloat(Minimum=3, Maximum=9)
     5=&RandomRangeFloat(Minimum=3, Maximum=9)
     Index=&MakeIntFromFloat(Float=$User.Variant)
     Num=::NewEnumerator4
    )
    UsePositionOffset=true
    Write Material Random=true
   )
   &AddVelocity(
    Module.Use Falloff Distance=false
    Override Seed=false
    Velocity Mode=::NewEnumerator1
    Velocity Speed=&RandomRangeFloat(Minimum=0, Maximum=50)
   ) {module=AddVelocity001}
   /SetVariables(Particles.VisibilityTag=&MergeHelper_IntSwitch(0=0, 1=0, 2=0, 3=1, 4=2, 5=2, Index=&MakeIntFromFloat(Float=$User.Variant), Num=::NewEnumerator4)) {isSetParameters=true}
   /SetVariables(Particles.RibbonID=/Niagara/DynamicInputs/NiagaraID/MakeNiagaraIDFromInt(
     Index:NiagaraInt32=/Niagara/DynamicInputs/Divide/Divide_Int(A:NiagaraInt32=&ReturnParticleIDAttributes(), B:NiagaraInt32=3)
     Acquire Tag:NiagaraInt32=0
    )) {isSetParameters=true}
  ]
  particleUpdateScript [
   &ParticleState()
   &AccelerationForce(Acceleration=<0 0 -100>)
   &Drag(Drag=10, Use Linear Drag=true, Use Rotational Drag=true)
   &CurlNoiseForce(Noise Strength=200, Pan Noise=false)
   &SolveForcesAndVelocity(Clamp Velocity=false, Limit Acceleration=false)
   &ScaleSpriteSize(Uniform Curve Sprite Scale=DataInterface() {propertyValues {Curve {keys [
        {arriveTangent=-1.3862, interpMode=RCIM_Cubic, time=0, value=1}
        {arriveTangent=-1.8417, interpMode=RCIM_Cubic, leaveTangent=-1.8417, tangentMode=RCTM_User, time=1, value=0}
       ]}}})
   &ScaleColor(
    Scale Alpha=&Multiply_Float(
     A=&FloatFromCurve(FloatCurve=DataInterface() {propertyValues {Curve {keys [<at 0.3 1>,<at 1 0>]}}})
     B=&FNTimeOfDayScalar(Night Scale=0.5, Morning Scale=0.75, Evening Scale=0.75)
    )
    ScaleA=true
    ScaleRGB=true
   )
   /SetVariables(
    Particles.GlowColor=&MakeLinearColorFromVectorAndFloat(
     "Vector (RGB)"=&MakeVectorFromLinearColorRGB(Color=&MergeHelper_ColorSwitch(
       0=&Multiply_LinearColorByFloat(
        Linear Color=&ColorFromCurve(
         ColorCurve=DataInterface() {propertyValues {
           AlphaCurve {keys [<at 0 1 RCIM_Cubic>]}
           BlueCurve {keys [
             <at 0 0.3101>
             {arriveTangent=0.3864, interpMode=RCIM_Cubic, time=0.1915, value=0.5}
             <at 0.5198 0.5>
             {arriveTangent=-0.3954, interpMode=RCIM_Cubic, leaveTangent=-0.3954, time=0.7638, value=0.5}
             <at 1 0.3101>
            ]}
           GreenCurve {keys [
             <at 0 0.5677>
             {arriveTangent=-0.8426, interpMode=RCIM_Cubic, time=0.1915, value=0.3594}
             <at 0.5198 0.1536>
             {arriveTangent=0.8622, interpMode=RCIM_Cubic, leaveTangent=0.8622, time=0.7638, value=0.0833}
             <at 1 0.5677>
            ]}
           RedCurve {keys [
             <at 0 0.1117>
             {arriveTangent=0.2598, interpMode=RCIM_Cubic, time=0.1915, value=0.0703}
             <at 0.5198 0.2394>
             {arriveTangent=-0.2659, interpMode=RCIM_Cubic, leaveTangent=-0.2659, time=0.7638, value=0.4431}
             <at 1 0.1117>
            ]}
          }}
         CurveIndex=&Frac_Float(Float=&Add_Float(A=$Particles.NormalizedAge, B=&RandomRangeFloat(Minimum=0, Maximum=-0.5)))
        )
        Float=&RandomRangeFloat(Minimum=0.2, Maximum=0.5)
       )
       1=&Multiply_LinearColorByFloat(
        Linear Color=&ColorFromCurve(
         ColorCurve=DataInterface() {propertyValues {
           AlphaCurve {keys [<at 0 1 RCIM_Cubic>]}
           BlueCurve {keys [
             <at 0.0284 0.8677>
             {arriveTangent=-0.4812, interpMode=RCIM_Cubic, leaveTangent=-0.4812, time=0.4737, value=0.8646}
             <at 0.7925 0.5>
             <at 0.9918 0.8646>
            ]}
           GreenCurve {keys [
             <at 0.0284 1>
             {arriveTangent=-1.2221, interpMode=RCIM_Cubic, leaveTangent=-1.2221, time=0.4737, value=0.1067}
             <at 0.7925 0.0662>
             <at 0.9918 0.7183>
            ]}
           RedCurve {keys [
             <at 0.0284 0>
             {arriveTangent=0.4161, interpMode=RCIM_Cubic, leaveTangent=0.4161, time=0.4737, value=0.231}
             <at 0.7925 0.3179>
             <at 0.9918 0>
            ]}
          }}
         CurveIndex=&Frac_Float(Float=&Add_Float(A=$Particles.NormalizedAge, B=&RandomRangeFloat(Minimum=0, Maximum=1)))
        )
        Float=&RandomRangeFloat(Minimum=0.2, Maximum=0.5)
       )
       2=?
       3=?
       Index=/Niagara/DynamicInputs/Subtract/Subtract_Integer(A:NiagaraInt32=&MakeIntFromFloat(Float=$User.Variant), B:NiagaraInt32=4)
       Num=::NewEnumerator0
      ))
     "Float (Alpha)"=&FloatFromCurve(FloatCurve=DataInterface() {propertyValues {Curve {keys [<at 0.5 1>,<at 1 1>]}}})
    )
    Particles.RibbonColor:LinearColor=&MakeLinearColorFromVectorAndFloat(
     "Vector (RGB)"=&MakeVectorFromLinearColorRGB(Color=&Multiply_LinearColorByFloat(
       Linear Color=&ColorFromCurve(
        ColorCurve=DataInterface() {propertyValues {
          AlphaCurve {keys [<at 0 1 RCIM_Cubic>]}
          BlueCurve {keys [<at 0.0378 0.8677>, {arriveTangent=-0.0033, interpMode=RCIM_Cubic, time=0.3642, value=0.8646}, <at 0.7623 0.5>, <at 0.9918 0.8646>]}
          GreenCurve {keys [<at 0.0378 1>, {arriveTangent=-0.2924, interpMode=RCIM_Cubic, time=0.3642, value=0.1067}, <at 0.7623 0.0662>, <at 0.9918 0.7183>]}
          RedCurve {keys [<at 0.0378 0>,<at 0.3642 0.231 RCIM_Cubic>,<at 0.7623 0.3179>,<at 0.9918 0>]}
         }}
        CurveIndex=&Frac_Float(Float=&Add_Float(
          A=&Multiply_Float(A=$Particles.NormalizedAge, B=0.3)
          B=&RandomRangeFloat(Minimum=0, Maximum=1)
         ))
       )
       Float=&RandomRangeFloat(Minimum=0.2, Maximum=0.5)
      ))
     "Float (Alpha)"=&FloatFromCurve(FloatCurve=DataInterface() {propertyValues {Curve {keys [<at 0.5 1>,<at 1 0>]}}}, Scale Curve=1.1)
    )
   ) {isSetParameters=true}
  ]
  rendererClasses [&NiagaraSpriteRendererProperties, &NiagaraRibbonRendererProperties]
  renderers [
   {rendererClass=&NiagaraSpriteRendererProperties}
   {rendererClass=&NiagaraSpriteRendererProperties, rendererIndex=1}
   {rendererClass=&NiagaraSpriteRendererProperties, rendererIndex=2}
   {rendererClass=&NiagaraRibbonRendererProperties, rendererIndex=3}
  ]
 }
}
summary {
 systemName=NS_Pickaxe_DashSurge_Swing
 emitters [
  {emitterName=Trail, rendererClasses [&NiagaraRibbonRendererProperties]}
  {
   emitterName=Variant_Sprites
   rendererClasses [&NiagaraSpriteRendererProperties, &NiagaraRibbonRendererProperties]
  }
 ]
 userVariables [
  User.bIsFrontEndPreview {type=NiagaraBool}
  User.RibbonIDOffset {type=NiagaraInt32}
  User.SwingSectionIndex {type=NiagaraInt32}
  User.Variant {type=NiagaraFloat, defaultValue=0, description=""}
 ]
}
