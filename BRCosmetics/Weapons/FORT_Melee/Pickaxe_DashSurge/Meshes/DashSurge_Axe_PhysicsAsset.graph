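# Physics asset for the DashSurge pickaxe mesh, giving its dynamic branch bones simulated collision bodies.
# A base joint and nine branch joints each carry one capsule.
# Constraints link the branches into three chains of three (1-2-3, 4-5-6, 7-8-9), each rooted at the base joint.
# Every constraint allows only one degree of swing and twist, so the branches stay nearly rigid.
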PhysicsAsset /BRCosmetics/Weapons/FORT_Melee/Pickaxe_DashSurge/Meshes/DashSurge_Axe_PhysicsAsset
bodies {
 dyn_base_joint [{
   shapeType=Capsule
   length=4.4756
   radius=1
   shapeName=dyn_base_joint_capsule
   center=<0.164 -0.0578 -5.6267>
   rotation=<rot -0.6204 -90 -3.4043>
  }]
 dyn_branch_1_joint [{
   shapeType=Capsule
   length=5.1345
   radius=1.3116
   shapeName=dyn_branch_1_joint_capsule
   center=<4.4394 0.1816 0.1551>
   rotation=<rot 1.8499 -89.4096 -85.6025>
  }]
 dyn_branch_2_joint [{
   shapeType=Capsule
   length=5.9502
   radius=1.5438
   shapeName=dyn_branch_2_joint_capsule
   center=<4.7658 0.413 -0.0566>
   rotation=<rot 0 -92.5463 -93.4178>
  }]
 dyn_branch_3_joint [{
   shapeType=Capsule
   length=5.9951
   radius=1.9936
   shapeName=dyn_branch_3_joint_capsule
   center=<5.9204 0.3153 -0.083>
   rotation=<rot 0 89.4547 -88.8521>
  }]
 dyn_branch_4_joint [{
   shapeType=Capsule
   length=3.1541
   radius=1.2826
   shapeName=dyn_branch_4_joint_capsule
   center=<3.017 -0.4196 0.0438>
   rotation=<rot 0.427 89.8861 -88.4061>
  }]
 dyn_branch_5_joint [{
   shapeType=Capsule
   length=4.5675
   radius=1.255
   shapeName=dyn_branch_5_joint_capsule
   center=<3.7877 -0.1548 0.0925>
   rotation=<rot 0.0166 90.7507 -90.1881>
  }]
 dyn_branch_6_joint [{
   shapeType=Capsule
   length=5.658
   radius=1.9966
   shapeName=dyn_branch_6_joint_capsule
   center=<5.1581 -0.2193 0.0473>
   rotation=<rot 0 89.7763 -88.4506>
  }]
 dyn_branch_7_joint [{
   shapeType=Capsule
   length=3.113
   radius=1.4168
   shapeName=dyn_branch_7_joint_capsule
   center=<3.8176 0.0849 -0.1684>
   rotation=<rot 0.1658 92.3873 -90.6182>
  }]
 dyn_branch_8_joint [{
   shapeType=Capsule
   length=5.9127
   radius=1.7706
   shapeName=dyn_branch_8_joint_capsule
   center=<5.4608 0.1324 -0.3758>
   rotation=<rot -0.2261 -86.8047 -87.415>
  }]
 dyn_branch_9_joint [{
   shapeType=Capsule
   length=6.8749
   radius=2.2257
   shapeName=dyn_branch_9_joint_capsule
   center=<6.7841 0.2772 -0.1916>
   rotation=<rot 0.0001 94.2779 -87.3087>
  }]
}
constraints [
 {bone1Name=dyn_branch_1_joint, bone2Name=dyn_base_joint, swing1LimitDegrees=1, swing2LimitDegrees=1, twistLimitDegrees=1}
 {bone1Name=dyn_branch_2_joint, bone2Name=dyn_branch_1_joint, swing1LimitDegrees=1, swing2LimitDegrees=1, twistLimitDegrees=1}
 {bone1Name=dyn_branch_3_joint, bone2Name=dyn_branch_2_joint, swing1LimitDegrees=1, swing2LimitDegrees=1, twistLimitDegrees=1}
 {bone1Name=dyn_branch_4_joint, bone2Name=dyn_base_joint, swing1LimitDegrees=1, swing2LimitDegrees=1, twistLimitDegrees=1}
 {bone1Name=dyn_branch_5_joint, bone2Name=dyn_branch_4_joint, swing1LimitDegrees=1, swing2LimitDegrees=1, twistLimitDegrees=1}
 {bone1Name=dyn_branch_6_joint, bone2Name=dyn_branch_5_joint, swing1LimitDegrees=1, swing2LimitDegrees=1, twistLimitDegrees=1}
 {bone1Name=dyn_branch_7_joint, bone2Name=dyn_base_joint, swing1LimitDegrees=1, swing2LimitDegrees=1, twistLimitDegrees=1}
 {bone1Name=dyn_branch_8_joint, bone2Name=dyn_branch_7_joint, swing1LimitDegrees=1, swing2LimitDegrees=1, twistLimitDegrees=1}
 {bone1Name=dyn_branch_9_joint, bone2Name=dyn_branch_8_joint, swing1LimitDegrees=1, swing2LimitDegrees=1, twistLimitDegrees=1}
]
