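# Animation Blueprint for the DashSurge pickaxe that adds secondary physics motion to the axe mesh.
# The input pose runs through two chained RigidBody nodes using DashSurge_Axe_PhysicsAsset, simulated relative to the handle bone with overridden gravity.
# The first, stronger pass is enabled only when bIsInFrontEnd is false; the second is damped through lower motion scales and clamps.
# The event graph holds only unwired stubs, so nothing here sets that flag.
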AnimBlueprint /BRCosmetics/Weapons/FORT_Melee/Pickaxe_DashSurge/Meshes/DashSurge_Axe_AnimBP
target_skeleton=/Game/Weapons/FORT_Melee/Meshes/SK_Melee_Skeleton
graphs [
 AnimGraph {nodes [
   AnimGraphNode_SubInput_0 = Input Pose()
   n1 = RigidBody!(
    ComponentPose=Local To Component(LocalPose=@AnimGraphNode_SubInput_0)
    bAlphaBoolEnabled:bool=NOT Boolean(A=Get bIsInFrontEnd())
    ExternalForce
   ) -> (Pose) {
    alphaInputType=Bool
    bClampLinearTranslationLimitToRefPose=true
    bOverrideWorldGravity=true
    overridePhysicsAsset=&/DashSurge_Axe_PhysicsAsset
    simulationSpace=BaseBoneSpace
    alphaBoolBlend {blendOption=Linear}
    baseBoneRef {boneName=Melee_TwoHanded_Handle}
    componentAppliedLinearAccClamp=<3000 3000 3000>
    componentLinearAccScale=<0.6 0.6 0.6>
    componentLinearVelScale=<0.6 0.6 0.6>
    overrideWorldGravity {z=-980}
    simSpaceSettings {maxAngularAcceleration=3000, maxAngularVelocity=3000, maxLinearAcceleration=3000, maxLinearVelocity=3000, worldAlpha=0.1}
   }
   n2 = RigidBody!(ComponentPose=@n1, Alpha:float, ExternalForce) -> (Pose) {
    bClampLinearTranslationLimitToRefPose=true
    bOverrideWorldGravity=true
    overridePhysicsAsset=&DashSurge_Axe_PhysicsAsset
    simulationSpace=BaseBoneSpace
    alphaBoolBlend {blendOption=Linear}
    baseBoneRef {boneName=Melee_TwoHanded_Handle}
    componentAppliedLinearAccClamp=<1000 1000 1000>
    componentLinearAccScale=<0.2 0.2 0.2>
    componentLinearVelScale=<0.2 0.2 0.2>
    overrideWorldGravity {z=-980}
    simSpaceSettings {maxAngularAcceleration=3000, maxAngularVelocity=3000, maxLinearAcceleration=3000, maxLinearVelocity=3000, worldAlpha=0.1}
   }
   Output Pose(Result=Component To Local(ComponentPose=@n2)) {layerGroup=None}
  ]}
 EventGraph {nodes [
   TryGetPawnOwner()
   Event BlueprintUpdateAnimation()
  ]}
]
variables [Is In Front End {type=boolean}]
